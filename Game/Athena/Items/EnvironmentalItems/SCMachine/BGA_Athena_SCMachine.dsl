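; Fortnite Battle Royale Reboot Van: an interactable, unkillable actor that resurrects eliminated squadmates.
; Spawn machine state drives screen materials, hologram, lights, audio and cues, while replicated flags sync interaction and disabled visuals.
; On resurrection, the server knocks nearby pawns and vehicles, grants storm-phase-tiered loot and skydives airborne players, then starts cooldown and credits quests.
; Enhanced resurrection effects wait for the game mode's player-spawn delegate to avoid being overwritten.
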(event Custom|EnabledState
  (bind _van (Variables|Default|GetVan))
  ;[1 1 1] Set Visually Enabled and turn on interactions
  (Components|Activation|Activate (Variables|Default|GetNS_SCMachine_HoloLogo_LW) true)
  (Rendering|SetVisibility (Variables|Default|GetHeadlightR) true false)
  (Variables|Interaction|SetbAllowInteract true)
  (Utilities|IsValid (Variables|/Script/Engine.StaticMeshComponent|GetStaticMesh (Variables|Default|GetVan))
    (:Is%20Valid
      (Rendering|Material|SetVectorParameterValueonMaterials _van (Variables|FX|GetVan%20Emissive%20Parameter) (Variables|Default|GetOriginalVanMaterialEmissive))
      (Rendering|Material|SetMaterial _van (Variables|FX|GetScreenMaterialIndex) (Variables|FX|GetScreensMaterial)))))

(event Custom|DisabledState
  (bind _van (Variables|Default|GetVan))
  ;[1 1 1] Set Visually Disabled State and turn off interact
  (Components|Activation|Deactivate (Variables|Default|GetNS_SCMachine_HoloLogo_LW))
  (Rendering|SetVisibility (Variables|Default|GetHeadlightR) false false)
  (Variables|Interaction|SetbAllowInteract false)
  (Utilities|IsValid (Variables|/Script/Engine.StaticMeshComponent|GetStaticMesh (Variables|Default|GetVan))
    (:Is%20Valid
      (Rendering|Material|SetVectorParameterValueonMaterials _van (Variables|FX|GetVan%20Emissive%20Parameter) "0, 0, 0")
      (Rendering|Material|CreateDynamicMaterialInstance _van (Variables|FX|GetScreenMaterialIndex) (Variables|FX|GetStaticScreen_Material) "None" :!class /Script/Engine.PrimitiveComponent))))

(event EventBeginPlay
  (|CallParentFunction)
  ;[1 0 0] Give ability to test DBNO/Reboot revive
  (Variables|/Script/FortniteGame.FortActorComponent_MultiInteract|SetGradualCounterMultiplier (Variables|Default|GetEnableDecayRow) (Variables|Default|GetMultiInteractActorComponent))
  (CallFunction|CheckforTetherRange)
  (Utilities|FlowControl|Sequence
    (:then_0
      (if (Networking|IsServer)
        (Class|FortAsyncActionCurrentPlaylistReady|PlaylistDataReadyAsync
          (:OnReady
            ;[1 1 1] Setup for Disabling the Van for modes that want the Van in but not usable
            (if (ScalableFloat|GetValueasBoolwithContext (Variables|Default|GetRow_EnabledInWorld))
              (Resurrection|SetSpawnMachineState "WaitingForUse")
              ;[1 1 1] Bind to know when the player has spawned. This is called AFTER the players have their initial shield and health set up. The callback is used to help set initial shields for enhanced resurrected players.
              (bind _asfort_game_mode_athena (Utilities|Casting|CastToFortGameModeAthena (Game|GetGameMode)))
              (EventDispatchers|BindEventtoOnPlayerSpawnDelegate _asfort_game_mode_athena (EventDispatchers|CreateEvent :!event HandlePlayerSpawned))
              (else
                (Variables|Default|SetRepDisabledState true)))))
        (else
          (CallFunction|InitMaterialParameters))))
    (:then_1
      (bind _multiinteractactorcomponent (Variables|Default|GetMultiInteractActorComponent)
        (:then
          (EventDispatchers|BindEventtoActorCountChangedDelegate _multiinteractactorcomponent (delegate Custom|HandleActorCountChanged)))
        (:else)))))

(event Custom|SkydiveIfAirborne
  (params (Player :type Fort%20Player%20Pawn%20Object%20Reference))
  (bind _self self)
  ;[1 1 1] Skydive rebooted player if they were spawned in the air (For things like rifts.)
  (Utilities|FlowControl|Delay (Variables|Default|GetDelayBeforeSkydiveCheck))
  (bind (_outhit _returnvalue) (Collision|LineTraceByChannel :Start (Transformation|GetActorLocation Player) :End (Math|Vector|vector+vector (Transformation|GetActorLocation Player) (Math|Vector|MakeVector 0.0 0.0 (Variables|Default|GetHeightCheckForSkydive))) :TraceChannel "TraceTypeQuery7" :bTraceComplex false :DrawDebugType "None" :bIgnoreSelf false :TraceColor "(R=1.000000,G=0.000000,B=0.000000,A=1.000000)" :TraceHitColor "(R=0.000000,G=1.000000,B=0.000000,A=1.000000)" :DrawTime 5.0))
  (Development|PrintString (Utilities|String|Append "Should Skydive: " (Utilities|String|ToString(Boolean) (and _returnvalue (Utilities|Equal(Object) (.HitActor (Collision|BreakHitResult _outhit)) _self)))))
  (if (not (and _returnvalue (Utilities|Equal(Object) (.HitActor (Collision|BreakHitResult _outhit)) _self)))
    (Parachute|BeginSkydiving Player false)))

(event Custom|HandleActorCountChanged
  (params (InteractingActorCount :type Integer) (MultiInteractMultiplier :type Float%20%28single-precision%29) (InteractingDuration :type Float%20%28single-precision%29))
  (bind _screensmaterial (Variables|FX|GetScreensMaterial))
  ;[1 0.6434 0.1679] Sync Screens with MultiItneract Component
  (Networking|SwitchHasAuthority
    (:Remote
      (if (Class|FortActorComponentMultiInteract|IsMultipleInteractionEnabled (Variables|Default|GetMultiInteractActorComponent))
        (Rendering|Material|SetScalarParameterValue _screensmaterial "BaseTime" InteractingDuration :!class /Script/Engine.MaterialInstanceDynamic)
        (Rendering|Material|SetScalarParameterValue _screensmaterial "InteractTime" (Utilities|Time|GetGameTimeinSeconds) :!class /Script/Engine.MaterialInstanceDynamic)
        (Rendering|Material|SetScalarParameterValue _screensmaterial "Multiplier" MultiInteractMultiplier :!class /Script/Engine.MaterialInstanceDynamic)))))

(event Resurrection|EventOnPlayerPawnResurrected
  (params (PlayerPawn :type Fort%20Player%20Pawn%20Object%20Reference) (bEnhancedResurrection :type Boolean))
  ;[1 0.2321 0.8847] Handle Player Pawn Resurrected (called only on server)
  (CallFunction|KnockPlayerPawns PlayerPawn)
  (CallFunction|KnockVehicles)
  (CallFunction|TriggerRespawnFX PlayerPawn bEnhancedResurrection)
  (CallFunction|GiveWeaponsandResources (Fortnite|GetFortPlayerControllerfromActor PlayerPawn) bEnhancedResurrection)
  (CallFunction|SkydiveifAirborne PlayerPawn)
  (Variables|Quest|SetRebootedPlayer PlayerPawn)
  ;;[1 1 1] After this call, the game mode applies the initial shield and health value to the respawned player. EnhancedResurrection may wanna set its own value and thus we're tracking the respawning players here so that we can set our values when the game mode tells us that the player spawn process is complete.
  ;; See HandlePlayerSpawned().
  (if bEnhancedResurrection
    (Utilities|Array|Add (Variables|Transient|GetEnhancedResurrectionRebootingPlayerPawns) PlayerPawn)
    (Utilities|FlowControl|Delay 0.0)
    (Ability|SendGameplayEventtoActor PlayerPawn "(TagName=\"Event.Player.Reboot.EnhancedResurrection\")")))

(event Custom|HandleAmbientSound
  (bind _ambientsound (Variables|Default|GetAmbientSound))
  (bind _ambientsound_2 (Variables|Default|GetAmbientSound))
  ;[1 1 1] Handle the changing of ambient audio
  (switch Utilities|FlowControl|Switch|SwitchonESpawnMachineState (Resurrection|GetSpawnMachineState)
    (:WaitingForUse
      (Audio|Components|Audio|FadeOut _ambientsound_2 1.0 0.0 "Linear" :!class /Script/Engine.AudioComponent)
      (Utilities|FlowControl|Delay 1.0)
      (Audio|Components|Audio|SetSound _ambientsound_2 (Variables|Sounds|GetDefaultAmbientAudio))
      (Audio|Components|Audio|FadeIn _ambientsound_2 4.0 1.0 0.0 "Linear" :!class /Script/Engine.AudioComponent))
    (:Active
      (Audio|Components|Audio|FadeOut (Variables|Default|GetAmbientSound) 1.0 0.0 "Linear" :!class /Script/Engine.AudioComponent))
    (:OnCooldown
      (Audio|Components|Audio|SetSound _ambientsound (Variables|Sounds|GetCooldownAmbientAudio))
      (Audio|Components|Audio|FadeIn _ambientsound 2.0 1.0 0.0 "Linear" :!class /Script/Engine.AudioComponent))))

(event Resurrection|EventOnResurrectionStarted
  (params (PlayerController :type Fort%20Player%20Controller%20Athena%20Object%20Reference))
  ;[1 1 1] Handle Resurrection Started
  (Variables|Interact|SetRepServerInteract false)
  (CallFunction|InteactComplete 0)
  (Resurrection|SetSpawnMachineState "Active"))

(event Resurrection|EventOnResurrectionCompleted
  (bind _multiinteractactorcomponent (Variables|Default|GetMultiInteractActorComponent))
  ;[1 1 1] Handle Resurrection Completed
  (Resurrection|SetSpawnMachineState "Complete")
  (Utilities|IsValid _multiinteractactorcomponent
    (:Is%20Valid
      (Class|FortActorComponentMultiInteract|ResetGradualInteractionTimer _multiinteractactorcomponent))
    (:Is%20Not%20Valid
      _))
  (Utilities|FlowControl|Delay (Variables|States|GetDelayBeforeCooldown)) ; Give the van a second before it switches to the cooldown state
  (CallFunction|StartCooldown)
  (CallFunction|CheckQuestCredit))

(event Custom|UpdateSquadTags
  (params (InteractingActor :type Fort%20Pawn%20Object%20Reference))
  ;[1 1 1] Remove "last remaining player in squad" tag
  (bind _result (ArrayMacros|SafeArrayGet (Variables|Quest|GetPlayersInteracting))
    (:%20%20
      (if (GameplayTags|HasMatchingGameplayTag (Game|Player|GetPlayerPawn _result) (Variables|Default|GetSoloSquadTag))
        (Ability|GameplayEffect|RemoveLooseGameplayTags (Game|Player|GetPlayerPawn _result) (Variables|Default|GetSoloSquadTags) true)))))

(event Custom|InteactComplete
  (params (InteractingPawn :type Fort%20Pawn%20Object%20Reference))
  ;[1 1 1] InteractComplete- knock things
  (CallFunction|KnockPlayerPawns 0)
  (CallFunction|KnockVehicles)
  (CallFunction|UpdateSquadTags InteractingPawn))

(event Custom|ToggleHologram
  (params (TurnOn :type Boolean))
  (bind _ns_scmachine_holologo_lw (Variables|Default|GetNS_SCMachine_HoloLogo_LW))
  (bind (_returnvalue_x _returnvalue_y _returnvalue_z) (Math|Vector|BreakVector (Transformation|GetWorldLocation _ns_scmachine_holologo_lw)))
  ;[1 1 1] Toggle Hologram -
  (if (Math|Boolean|NotEqual(Boolean) TurnOn (Components|Activation|IsActive _ns_scmachine_holologo_lw))
    (Ability|ExecuteGameplayCueWithParamsReplicated (Variables|Default|GetTransitionCue) (Ability|GameplayCue|MakeGameplayCueParameters :NormalizedMagnitude 0.0 :RawMagnitude 0.0 :EffectContext "()" :AggregatedSourceTags "(GameplayTags=)" :AggregatedTargetTags "(GameplayTags=)" :Location (Math|Vector|MakeVectorNetQuantize10 _returnvalue_x _returnvalue_y _returnvalue_z) :Normal "0,0,0" :GameplayEffectLevel 1 :AbilityLevel 1 :bReplicateLocationWhenUsingMinimalRepProxy false))
    (Components|Activation|SetActive _ns_scmachine_holologo_lw TurnOn false)
    (Networking|SwitchHasAuthority
      (:Authority
        (Networking|ForceNetUpdate)))))

(event Custom|SpawnRift
  (params (InFortPawn :type Fort%20Pawn%20Object%20Reference)))

(event Custom|Reset%20Rift)

(event Custom|TriggerRespawnFX
  (params (PhaseInPawn :type Fort%20Pawn%20Object%20Reference) (bEnhancedResurrection :type Boolean))
  (bind _returnvalue (select bEnhancedResurrection "(TagName=\"SCMachine.Respawn.EnhancedResurrection\")" 0))
  (bind _returnvalue_2 (GameplayTags|MakeGameplayTagContainerfromTag _returnvalue))
  (bind (_returnvalue_x _returnvalue_y _returnvalue_z) (Math|Vector|BreakVector (Transformation|GetWorldLocation (Variables|Default|GetNS_SCMachine_PedestalGlow_LW))))
  ;[1 1 1] TriggerRespawnFX- Hide Mannequin. play Mannequin FX and spawn in the bot.
  (Default|CallOnPawnResurrected PhaseInPawn bEnhancedResurrection)
  (if (Networking|IsServer)
    (if (not (Variables|Default|GetUseRiftSpawn))
      (bind _returnvalue_3 (Game|SpawnActorfromClass "/Game/Blueprints/Adventures/SpawnTeleporter/BP_VictoryDrone.BP_VictoryDrone_C" (Math|Transform|MakeTransform (Transformation|GetWorldLocation (Variables|Default|GetPadCollision)) (Transformation|GetActorRotation)) "AlwaysSpawn" "OverrideRootScale" 0 true true 0 true false false 0))
      (Actor|SetLifeSpan _returnvalue_3 2.25)
      (Transformation|AttachActorToActor _returnvalue_3 PhaseInPawn "None" "SnapToTarget" "SnapToTarget" "KeepRelative" true))
    (else
      ;[1 1 1] Play Sound Indicator and update screens on clients
      (bind _returnvalue_4 (Development|GetConsoleVariableFloatValue "SpawnMachine.SoundIndicator.MaxDistanceMultiplier"))
      (SoundIndicator|ShowSoundIndicator (Variables|Default|GetFortSoundIndicatorComponent) (Transformation|GetActorLocation) "/Game/Athena/Sounds/SC_Machine/SC_Machine_Complete_Cue.SC_Machine_Complete_Cue" _returnvalue_4 0)
      (CallFunction|UpdateScreen)
      (CallFunction|SpawnRift PhaseInPawn)))
  (Utilities|FlowControl|Delay (Variables|FX|GetFXDelayThenBotSpawn))
  (Ability|ExecuteGameplayCueWithParamsReplicated (Variables|Default|GetActivateCue) (Ability|GameplayCue|MakeGameplayCueParameters :NormalizedMagnitude 0.0 :RawMagnitude 0.0 :AggregatedSourceTags _returnvalue_2 :AggregatedTargetTags "(GameplayTags=)" :Location (Math|Vector|MakeVectorNetQuantize10 _returnvalue_x _returnvalue_y _returnvalue_z) :Normal "0,0,0" :SourceObject PhaseInPawn :GameplayEffectLevel 1 :AbilityLevel 1 :bReplicateLocationWhenUsingMinimalRepProxy false))
  (bind _asplayer_pawn_athena (Utilities|Casting|CastToPlayerPawn_Athena PhaseInPawn))
  (GameplayCue|ExecuteGameplayCue%28Burst%29Local PhaseInPawn (Variables|Default|GetSCMachine%20Respawn%20Cue) (Ability|GameplayCue|MakeGameplayCueParameters :NormalizedMagnitude 0.0 :RawMagnitude 0.0 :AggregatedSourceTags _returnvalue_2 :Location "0,0,0" :Normal "0,0,0" :GameplayEffectLevel 1 :AbilityLevel 1 :bReplicateLocationWhenUsingMinimalRepProxy false))
  (Class|PlayerPawnAthena|ActivateInvincibleVisuals _asplayer_pawn_athena (Variables|Default|GetInvincibleVisualsDuration))
  ;[1 1 1] Do a little warp of the van to make it look more springy
  (Components|Timeline|SetPlayRate (Variables|BGA_Athena_SCMachine|GetWarpVan) (Variables|FX|GetPlayrate))
  (bind (_direction _scale) (Timeline|WarpVan PlayFromStart)
    (:Update
      (Transformation|SetWorldScale3D (Variables|Default|GetVan) _scale))))

(event Custom|Start%20Activate%20Sound%20Timer
  (bind _activatesound (Variables|Default|GetActivateSound))
  ;[1 1 1] End activate audio
  (Utilities|FlowControl|RetriggerableDelay (ScalableFloat|GetValueatLevelwithContext (Variables|Rows|GetRow_InteractTime)))
  (SoundIndicator|StopSoundIndicatorLooping (Variables|Default|GetFortSoundIndicatorComponent))
  (Utilities|FlowControl|Sequence
    (:then_0
      (if (Audio|Components|Audio|IsPlaying _activatesound)
        (Audio|Components|Audio|Play (Variables|Default|GetActivateEndSound) 0.0)))
    (:then_1
      (Audio|Components|Audio|FadeOut _activatesound 0.5 0.0 "Linear" :!class /Script/Engine.AudioComponent))))

(event Custom|StartCooldown
  (bind _returnvalue (Resurrection|GetCooldownLength))
  ;[1 1 1] Start cooldown after final respawn
  (Resurrection|SetSpawnMachineState "OnCooldown")
  (if (> _returnvalue 0)
    (bind _returnvalue_2 (Utilities|Time|SetTimerbyEvent (delegate Custom|CooldownTimer) _returnvalue))
    (Variables|States|SetTimer_Cooldown _returnvalue_2)
    (else
      (CallFunction|CooldownTimer))))

(event Custom|Update%20Cooldown%20Sound
  (if (Utilities|Enum|Equal(Enum) (Resurrection|GetSpawnMachineState) "OnCooldown")
    (Audio|Components|Audio|FadeIn (Variables|Default|GetSC_Machine_Cooldown_Cue) 0.15 1.0 0.0 "Linear" :!class /Script/Engine.AudioComponent)
    (elif (Audio|Components|Audio|IsPlaying (Variables|Default|GetSC_Machine_Cooldown_Cue))
      (Audio|Components|Audio|FadeOut (Variables|Default|GetSC_Machine_Cooldown_Cue) 0.15 0.0 "Linear" :!class /Script/Engine.AudioComponent)
      (Audio|Components|Audio|Play (Variables|Default|GetSC_Machine_Cooldown_Cue_End) 0.0))))

(event Resurrection|EventOnBeginLongUse
  (params (AthenaPC :type Fort%20Player%20Controller%20Athena%20Object%20Reference))
  ;[1 1 1] Replicated Interact
  (Utilities|Array|AddUnique (Variables|Quest|GetPlayersInteracting) AthenaPC)
  (Development|PrintString "Event - On Begin Long use")
  (Variables|Interact|SetRepServerInteract true))

(event Custom|CooldownTimer
  ;[1 1 1] CooldownTimerEnd
  (Resurrection|SetSpawnMachineState "WaitingForUse")
  (Ability|ExecuteGameplayCueReplicated (Variables|Default|GetEndCue))
  (CallFunction|ResetRift))

(event Resurrection|EventOnInterruptLongUse
  (params (AthenaPC :type Fort%20Player%20Controller%20Athena%20Object%20Reference))
  ;[1 1 1] Replicated Interact
  (Utilities|Array|RemoveItem (Variables|Quest|GetPlayersInteracting) AthenaPC)
  (Development|PrintString "Event - On Interrupt Long Use")
  ;[1 0.6602 0.2944] When MultiInteract, only notify stop interaction when no players interacting
  (if (Utilities|IsArrayNonEmpty (Variables|Quest|GetPlayersInteracting))
    (if (Class|FortActorComponentMultiInteract|IsMultipleInteractionEnabled (Variables|Default|GetMultiInteractActorComponent))
      _
      (else
        _)))
  (Variables|Interact|SetRepServerInteract false))

(event Interaction|EventBlueprintOnBeginInteract
  (params (InteractingPawn :type Fort%20Pawn%20Object%20Reference) (InteractionBeingAttempted :type EInteractionBeingAttempted%20Enum :default "FirstInteraction") (InteractionType :type TInteractionType%20Enum :default "IT_NoInteraction"))
  ;[0.04597 0.1071 0.15 0.5] BlueprintOnBeginInteract: Play buy reboot card interact sound when a player begins the secondary interaction.
  (|CallParentFunction InteractingPawn InteractionBeingAttempted InteractionType)
  (if (Utilities|Enum|Equal(Enum) InteractionBeingAttempted "SecondInteraction")
    (Audio|PlaySound2D (Variables|BuyRebootCard|GetBuyRebootCardInteractSound))))

(event EventDestroyed
  ;[1 0.63 0] Destroyed
  (|CallParentFunction)
  (bind _multiinteractactorcomponent (Variables|Default|GetMultiInteractActorComponent)
    (:then
      (EventDispatchers|UnbindEventfromActorCountChangedDelegate _multiinteractactorcomponent (EventDispatchers|CreateEvent :!event Custom|HandleActorCountChanged)))
    (:else)))

(event OnComponentBeginOverlap(BlockingVolume)
  (params (OverlappedComponent :type Primitive%20Component%20Object%20Reference) (OtherActor :type Actor%20Object%20Reference) (OtherComp :type Primitive%20Component%20Object%20Reference) (OtherBodyIndex :type Integer) (bFromSweep :type Boolean) (SweepResult :type Hit%20Result%20Structure%20%28by%20ref%29)))

(event OnComponentEndOverlap(BlockingVolume)
  (params (OverlappedComponent :type Primitive%20Component%20Object%20Reference) (OtherActor :type Actor%20Object%20Reference) (OtherComp :type Primitive%20Component%20Object%20Reference) (OtherBodyIndex :type Integer)))

(event EventEndPlay
  (params (EndPlayReason :type EEndPlayReason%20Enum :default "Destroyed"))
  ;[0.15 0.08481 0.01615 0.5] EndPlay: Clean up binds.
  (Utilities|FlowControl|Sequence
    (:then_0
      (bind _asfort_game_mode_athena (Utilities|Casting|CastToFortGameModeAthena (Game|GetGameMode)))
      (EventDispatchers|UnbindEventfromOnPlayerSpawnDelegate _asfort_game_mode_athena (EventDispatchers|CreateEvent :!event HandlePlayerSpawned)))
    (:then_1
      (|CallParentFunction EndPlayReason))))

(event Interaction|EventBlueprintOnInteract
  (params (InteractingPawn :type Fort%20Pawn%20Object%20Reference) (InteractionBeingAttempted :type EInteractionBeingAttempted%20Enum :default "FirstInteraction"))
  ;[0.03056 0.1269 0.15 0.5] OnInteract: On a successful second interact, send gameplay even to player. Used for quest tracking.
  (|CallParentFunction InteractingPawn InteractionBeingAttempted)
  (if (Utilities|Enum|Equal(Enum) InteractionBeingAttempted "SecondInteraction")
    (Ability|SendGameplayEventtoActor InteractingPawn "(TagName=\"Event.Player.Reboot.BuyRebootCard\")")))

(event Resurrection|EventOnSpawnMachineStateChanged
  (bind _returnvalue (Resurrection|GetSpawnMachineState))
  (bind _returnvalue_2 (Resurrection|GetSpawnMachineState))
  (bind _returnvalue_3 (Utilities|Enum|NotEqual(Enum) _returnvalue "OnCooldown"))
  (bind _returnvalue_4 (Utilities|Enum|NotEqual(Enum) _returnvalue_2 "OnCooldown"))
  (Development|PrintString (Utilities|String|EnumtoString (Resurrection|GetSpawnMachineState)))
  (CallFunction|HandleAmbientSound)
  (CallFunction|UpdateScreen)
  ;[1 1 1] Set the state of all the visuals based on the van state. Need to do this every time for replays so they change correctly.
  (Components|Activation|SetActive (Variables|Default|GetNS_SCMachine_PedestalGlow_LW) _returnvalue_4 false) ; Show FX glow
  (Rendering|SetVisibility (Variables|Default|GetHeadlightR) _returnvalue_4 false)
  (Rendering|Material|SetVectorParameterValueonMaterials (Variables|Default|GetVan) "Emissive Multiplier" (select _returnvalue_4 "2.000000,2.000000,2.000000" "0, 0, 0"))
  (CallFunction|ToggleHologram (or (Utilities|Enum|Equal(Enum) (Resurrection|GetSpawnMachineState) "OnCooldown") (Utilities|Enum|Equal(Enum) (Resurrection|GetSpawnMachineState) "WaitingForUse")))
  (if (Utilities|Enum|Equal(Enum) (Resurrection|GetSpawnMachineState) "Active")
    (Ability|AddGameplayCueReplicated :self self :GameplayCueTag (Variables|Default|GetActiveCue))
    (Utilities|Time|ClearandInvalidateTimerbyHandle (Variables|Default|GetRemoveActiveGameplayCueTimerHandle))
    ;[1 1 1] Flash lights when spawning starts
    (Effects|SpawnEmitterAttached :EmitterTemplate "/Game/Athena/Items/EnvironmentalItems/SCMachine/FX/P_SCMachine_LightFlashes.P_SCMachine_LightFlashes" :AttachToComponent (Variables|Default|GetVan) :AttachPointName "None" :Scale "1.000000,1.000000,1.000000" :LocationType "SnapToTargetIncludingScale" :bAutoDestroy true :PoolingMethod "None" :bAutoActivate true)
    (Utilities|FlowControl|Sequence
      (:then_0
        (bind (_direction _lightbrightness) (Timeline|LightFlash Play)
          (:Update
            (Rendering|Material|SetVectorParameterValue (Variables|FX|GetVanBodyMaterial) "Emissive Multiplier" (Math|Color|SelectColor :A (Math|Conversions|ToLinearColor(Float) (* _lightbrightness 2.0)) :bPickA _returnvalue_3) :!class /Script/Engine.MaterialInstanceDynamic)
            (Rendering|Components|Light|SetIntensity (Variables|Default|GetHeadlightL) (Math|Float|SelectFloat (* _lightbrightness (Variables|FX|GetHeadlightIntensity)) (Variables|FX|GetHeadlightIntensity) _returnvalue_3)))
          (:Finished
            (Variables|Sound|SetDid%20Horn%20Play? false))))
      (:then_1
        (CallFunction|PlayHornSound)))
    (elif (Utilities|Time|IsTimerActivebyHandle (Variables|Default|GetRemoveActiveGameplayCueTimerHandle))
      _
      (else
        (bind _returnvalue_5 (Utilities|Time|SetTimerbyEvent (delegate Custom|RemoveActiveGameplayCue) (Variables|Default|GetRemoveActiveGameplayCueDelay)))
        (Variables|Default|SetRemoveActiveGameplayCueTimerHandle _returnvalue_5))))
  (CallFunction|UpdateCooldownSound))

(event Custom|RemoveActiveGameplayCue
  ;[1 1 1] Set the state of all the visuals based on the van state. Need to do this every time for replays so they change correctly.
  (Ability|RemoveGameplayCueReplicated (Variables|Default|GetActiveCue)))

(event Custom|Play%20Horn%20Sound
  (bind _hornalertaudio (Variables|Sounds|GetHornAlertAudio))
  ;[1 1 1] Play Sound for Headlights Flash
  (if (Utilities|Enum|Equal(Enum) (Resurrection|GetSpawnMachineState) "Active")
    (if (not (Variables|Sound|GetDid%20Horn%20Play?))
      (Audio|PlaySoundatLocation _hornalertaudio (Transformation|GetActorLocation))
      (bind _returnvalue (Development|GetConsoleVariableFloatValue "SpawnMachine.SoundIndicator.MaxDistanceMultiplier"))
      (SoundIndicator|ShowSoundIndicator (Variables|Default|GetFortSoundIndicatorComponent) (Transformation|GetActorLocation) _hornalertaudio _returnvalue 0)
      (Variables|Sound|SetDid%20Horn%20Play? true))))

(fn BlueprintCanInteract
  (params (InteractingPawn :type Fort%20Pawn%20Object%20Reference) (InteractionBeingAttempted :type EInteractionBeingAttempted%20Enum :default "FirstInteraction") (InteractionType :type TInteractionType%20Enum :default "IT_NoInteraction"))
  (return (Math|Boolean|ORBoolean (Interaction|CanPlayerSquadInteract InteractingPawn InteractionType) (Utilities|Enum|Equal(Enum) (Resurrection|GetSpawnMachineState) "WaitingForUse") (Math|Boolean|ANDBoolean (BuyRebootCard|IsBuyingRebootCardsEnabled) (Utilities|Enum|Equal(Enum) InteractionBeingAttempted "SecondInteraction") (Utilities|Enum|Equal(Enum) (Resurrection|GetSpawnMachineState) "OnCooldown")))))

(fn BlueprintGetFailedInteractionString
  (params (InteractingPawn :type Fort%20Pawn%20Object%20Reference) (InteractionBeingAttempted :type EInteractionBeingAttempted%20Enum :default "FirstInteraction"))
  ;[1 1 1] Set Failed Text Based on State
  (switch Utilities|FlowControl|Switch|SwitchonESpawnMachineState (Resurrection|GetSpawnMachineState)
    (:Default
      _)
    (:WaitingForUse
      _)
    (:BeingUsed
      _)
    (:Active
      (LocalVariables|SetFailedText (Variables|Texts|GetFailedTextRebooting))
      (return (LocalVariables|GetFailedText)))
    (:Complete
      (LocalVariables|SetFailedText (Variables|Texts|GetFailedTextRecharging) :!label _label)
      (return (LocalVariables|GetFailedText)))
    (:OnCooldown
      (goto _label)))
  (LocalVariables|SetFailedText (Variables|Texts|GetFailedTextNoCard))
  (return (LocalVariables|GetFailedText)))

(fn BlueprintGetInteractionString
  (params (InteractingPawn :type Fort%20Pawn%20Object%20Reference) (InteractionBeingAttempted :type EInteractionBeingAttempted%20Enum :default "FirstInteraction"))
  (return (Variables|Texts|GetPassedText)))

(fn BlueprintGetInteractionTime
  (params (InteractingPawn :type Fort%20Pawn%20Object%20Reference) (InteractionBeingAttempted :type EInteractionBeingAttempted%20Enum :default "FirstInteraction"))
  ;[0.5412 1 0.3843] This determines the interaction time for rebooting allies, the second interaction time is controlled natively.
  ;
  ; Check the runtime registry because the value can be changed in Creative Royale.
  (return (ScalableFloat|GetValueatLevelwithContext (Variables|Rows|GetRow_InteractTime)) true))

(fn BlueprintUpdateDynamicInteractionOptions
  (params (InteractingPawn :type Fort%20Pawn%20Object%20Reference) (InteractionBeingAttempted :type EInteractionBeingAttempted%20Enum :default "FirstInteraction") (InteractionType :type TInteractionType%20Enum :default "IT_NoInteraction"))
  (bind _outgameplaytags (|CallParentFunction InteractingPawn InteractionBeingAttempted InteractionType))
  (switch Utilities|FlowControl|Switch|SwitchonESpawnMachineState (Resurrection|GetSpawnMachineState)
    (:Default
      _)
    (:WaitingForUse
      _)
    (:BeingUsed
      _)
    (:Active
      (GameplayTags|AddGameplayTag _outgameplaytags "(TagName=\"HUD.Athena.InteractionOptions.DimPrimaryContext\")" :!class /Script/GameplayTags.BlueprintGameplayTagLibrary :!label _label))
    (:Complete
      (goto _label))
    (:OnCooldown
      (goto _label)))
  (return _outgameplaytags))

(fn BuildDataRegistryResolverScope
  (bind _self self)
  (bind _returnvalue (Minigame|GetMinigame _self :!class /Script/FortniteGame.MinigameBlueprintLibrary))
  ;[1 0.8475 0.2776] Gives the name of the volume as a scope to the data registry. Needed for calls to GetValueWithContext to work
  (Utilities|IsValid _returnvalue
    (:Is%20Valid
      (return (Utilities|Array|MakeArray (SpatialGameplay|GetSpatialGameplayInterfaceFNameforActor _returnvalue)) 0 true))
    (:Is%20Not%20Valid
      (return 0 0 false))))

(fn CheckForTetherRange
  (bind (_boxextent_x _boxextent_y _boxextent_z) (Math|Vector|BreakVector (Variables|/Script/Engine.BoxComponent|GetBoxExtent_X (Variables|Default|GetInteractionBox))))
  (if (Resurrection|UseResurrectionTetherRange)
    (bind _output_get (Variables|SetLocalBoxExtent_X (Math|Vector|MakeVector (* _boxextent_x (ScalableFloat|GetScalableFloatMultiplier (Variables|Resurrection|Interaction|GetResurrectionHorizontalTetherMultiplier))) (* _boxextent_y (ScalableFloat|GetScalableFloatMultiplier (Variables|Resurrection|Interaction|GetResurrectionHorizontalTetherMultiplier))) (* _boxextent_z (ScalableFloat|GetScalableFloatMultiplier (Variables|Resurrection|Interaction|GetResurrectionVerticalTetherMultiplier))))))
    (Components|Box|SetBoxExtent (Variables|Default|GetInteractionBox) _output_get false)))

(fn CheckQuestCredit
  (bind _playersinteracting (Variables|Quest|GetPlayersInteracting))
  (bind _rebootedplayer (Variables|Quest|GetRebootedPlayer))
  ;[1 1 1] Send a quest event on instigator(s) when the player has been revived and quests can be completed (not creative mode)
  (bind _cancompletequest (Class|AthenaQuestFunctionLibrary|AthenaCheckCreativeMode (Game|GetGameState)))
  (if _cancompletequest
    (Utilities|FlowControl|Sequence
      (:then_0
        (for _array_element _playersinteracting
          (bind _output_get (LocalVariables|SetFortPlayerController_Local _array_element))
          (Utilities|IsValid (Quests|GetQuestManager _output_get "Invalid")
            (:Is%20Valid
              (Quest|SendComplexCustomStatEvent (Quests|GetQuestManager _output_get "Invalid") 0 "(GameplayTags=)" "(GameplayTags=((TagName=\"Athena.Quests.SecondChanceMachineRevive\")))" 1))))
        (Utilities|Array|Clear (Variables|Quest|GetPlayersInteracting)))
      (:then_1
        ;;[0 0 0] Apply a temporary tag to the target player indicating that they've recently been rebooted
        (Utilities|IsValid _rebootedplayer
          (:Is%20Valid
            (GameplayEffects|ApplyGameplayEffectToSelf :self (Ability|GetFortAbilitySystemComponent _rebootedplayer) :GameplayEffectClass "/Game/Athena/Items/EnvironmentalItems/SCMachine/GE_Athena_PostReboot_ApplyTag.GE_Athena_PostReboot_ApplyTag_C" :Level 0.0)))))
    (else
      (Utilities|Array|Clear _playersinteracting))))

(fn GetCacheContextOverride
  ;[0.3642 0.9801 1] We don't need to return any thing because we don't need to override the CacheContext.
  (return 0))

(fn GetHoloLogoNiagaraComponent
  (return (Variables|Default|GetNS_SCMachine_HoloLogo_LW)))

(fn GetOverrideItemInteractAbilityTag
  (return "(TagName=\"Gameplay.Action.Player.InteractSearch.Reboot\")"))

(fn GiveWeaponsAndResources
  (params (FortPC :type Fort%20Player%20Controller%20Object%20Reference) (bEnhancedResurrection :type Boolean))
  ;[0.15 0.06724 0.1493 0.5] GiveWeaponsAndResources: Player has respawned, give them the desired loot.
  (Utilities|IsValid FortPC
    (:Is%20Valid
      (Utilities|FlowControl|Sequence
        (:then_0
          (bind _returnvalue (SafeZone|GetCurrentSafeZonePhase))
          (if (< _returnvalue 3)
            (LocalVariables|SetTierGroupName (Variables|Loot|GetGrantedItemsLootTier))
            (elif (< _returnvalue 5)
              (LocalVariables|SetTierGroupName (Variables|Default|GetMid%20Reboot))
              (else
                (LocalVariables|SetTierGroupName (Variables|Default|GetLate%20Reboot)))))
          (if bEnhancedResurrection
            (bind _returnvalue_2 (Fortnite|Athena|GameState|GetGameStateAthena))
            (Utilities|IsValid _returnvalue_2
              (:Is%20Valid
                (bind _returnvalue_3 (Actor|GetComponentbyClass _returnvalue_2 "/Script/FortniteGame.FortGameStateComponent_Reboot"))
                (Utilities|IsValid _returnvalue_3
                  (:Is%20Valid
                    (LocalVariables|SetTierGroupName (Resurrection|GetEnhancedResurrectionLootTierGroupName _returnvalue_3))))))))
        (:then_1
          (bind (_outloottodrop _returnvalue_4) (Loot|PickLootDrops (LocalVariables|GetTierGroupName) 0 -1 "(GameplayTags=)"))
          (if _returnvalue_4
            (for _array_element _outloottodrop
              (bind (_array_element_itemdefinition _array_element_itemguid _array_element_count) (Class|FortItemFunctionLibrary|BreakFortItemEntry _array_element))
              (bind _asfort_world_item_definition (Utilities|Casting|CastToFortWorldItemDefinition _array_element_itemdefinition))
              (Inventory|GiveItemtoInventoryOwner :InventoryOwner FortPC :ItemDefinition _asfort_world_item_definition :NumberToGive _array_element_count :bNotifyPlayer true :ItemLevel -1 :PickupInstigatorHandle -1 :bUseItemPickupAnalyticEvent false :WeaponAmmoOverride -1))))))))

(fn HandlePlayerSpawned
  (params (Pawn :type Fort%20Player%20Pawn%20Object%20Reference))
  ;[0.0273 0.15 0.1433 0.5] HandlePlayerSpawned: Apply enhanced resurrection gameplay effects to the respawned player.
  (Networking|SwitchHasAuthority
    (:Authority
      (Utilities|IsValid Pawn
        (:Is%20Valid
          (bind _returnvalue (Utilities|Array|RemoveItem (Variables|Transient|GetEnhancedResurrectionRebootingPlayerPawns) Pawn))
          (if _returnvalue
            (Fortnite|ApplyGameplayEffectsToActor Pawn (Variables|EnhancedResurrection|GetEnhancedResurrectionGameplayEffects) (0.0 0.0 0.0) false 0 (Ability|GetFortAbilitySystemComponent Pawn) "()")))))))

(fn InitMaterialParameters
  (bind _van (Variables|Default|GetVan))
  (bind _mid (Material|GetMID _van))
  (Utilities|IsValid _mid
    (:Is%20Valid
      (bind _returnvalue (Rendering|Material|GetVectorParameterValue _mid (Variables|FX|GetVan%20Emissive%20Parameter) :!class /Script/Engine.MaterialInstanceDynamic))
      (Variables|Default|SetOriginalVanMaterialEmissive (Math|Conversions|ToVector(LinearColor) _returnvalue)))))

(fn KnockPlayerPawns
  (params (IgnoreKnock :type Actor%20Object%20Reference))
  ;[1 1 1] KnockPlayers
  (for _array_element (Collision|GetOverlappingActors (Variables|Default|GetBlockingVolume) "/Script/FortniteGame.FortPlayerPawn" :!class /Script/Engine.PrimitiveComponent)
    (if (Utilities|NotEqual(Object) _array_element (Variables|GetIgnoreKnock))
      (bind _asfort_player_pawn (Utilities|Casting|CastToFortPlayerPawn _array_element))
      (Character|LaunchCharacter _asfort_player_pawn (Math|Vector|MakeVector (.x (* (Math|Vector|RotateVectorAroundAxis (Transformation|GetActorForwardVector) (Math|Random|RandomFloatinRange 0.0 3.6e+02) "0.000000, 0,1.000000") (Variables|Knockback|GetXYKnockAmountPlayer))) (.y (* (Math|Vector|RotateVectorAroundAxis (Transformation|GetActorForwardVector) (Math|Random|RandomFloatinRange 0.0 3.6e+02) "0.000000, 0,1.000000") (Variables|Knockback|GetXYKnockAmountPlayer))) (Variables|Knockback|GetZKnockAmountPlayer)) true true))))

(fn KnockVehicles
  (bind _zknockamountvehicle (Variables|Knockback|GetZKnockAmountVehicle))
  (bind _xyknockamountvehicle (Variables|Knockback|GetXYKnockAmountVehicle))
  (bind _returnvalue (Math|Random|RandomFloatinRange 0.0 3.6e+02))
  (bind _returnvalue_2 (Transformation|GetActorForwardVector))
  (bind _returnvalue_3 (Math|Vector|RotateVectorAroundAxis _returnvalue_2 _returnvalue "0.000000, 0,1.000000"))
  (bind _returnvalue_4 (* _returnvalue_3 _xyknockamountvehicle))
  (bind _y (.y _returnvalue_4))
  (bind _x (.x _returnvalue_4))
  (bind _returnvalue_5 (Math|Vector|MakeVector _x _y _zknockamountvehicle))
  (Utilities|FlowControl|Sequence
    (:then_0
      ;[1 1 1] AthenaVehicle
      (for _array_element (Collision|GetOverlappingActors (Variables|Default|GetBlockingVolume) "/Script/FortniteGame.FortAthenaVehicle" :!class /Script/Engine.PrimitiveComponent)
        (Development|PrintString "KnockVehicle")
        (bind _asfort_vehicle_interface (Utilities|Casting|CastToFortVehicleInterface _array_element))
        (Audio|BroadcastAbilityImpactatLocation 0 (Vehicle|GetActor _asfort_vehicle_interface) _returnvalue_5 777.0)))
    (:then_1
      ;[1 1 1] CharacterVehicle
      (for _array_element_2 (Collision|GetOverlappingActors (Variables|Default|GetBlockingVolume) "/Script/FortniteGame.FortCharacterVehicle" :!class /Script/Engine.PrimitiveComponent)
        (Development|PrintString "KnockVehicle")
        (bind _asfort_vehicle_interface_2 (Utilities|Casting|CastToFortVehicleInterface _array_element_2))
        (Audio|BroadcastAbilityImpactatLocation 0 (Vehicle|GetActor _asfort_vehicle_interface_2) _returnvalue_5 777.0)))))

(fn OnRep_RepDisabledState
  (if (Variables|Default|GetRepDisabledState)
    (CallFunction|DisabledState)
    (else
      (CallFunction|EnabledState))))

(fn OnRep_RepServerInteract
  (bind _returnvalue (Utilities|Time|GetGameTimeinSeconds))
  (bind _row_interacttime (Variables|Rows|GetRow_InteractTime))
  (bind _returnvalue_2 (ScalableFloat|GetValueatLevelwithContext _row_interacttime))
  (bind _ns_scmachine_holologo_lw (Variables|Default|GetNS_SCMachine_HoloLogo_LW))
  (bind _activatesound (Variables|Default|GetActivateSound))
  (bind _screensmaterial (Variables|FX|GetScreensMaterial))
  (if (Variables|Interact|GetRepServerInteract)
    (Utilities|FlowControl|Sequence
      (:then_0
        (Effects|Components|ParticleSystem|SetFloatParameter _ns_scmachine_holologo_lw "GameTime" _returnvalue)
        (Effects|Components|ParticleSystem|SetFloatParameter _ns_scmachine_holologo_lw "InteractDuration" _returnvalue_2))
      (:then_1
        (Rendering|Material|SetScalarParameterValue _screensmaterial "InteractTime" _returnvalue :!class /Script/Engine.MaterialInstanceDynamic)
        (Rendering|Material|SetScalarParameterValue _screensmaterial "InteractDuration" _returnvalue_2 :!class /Script/Engine.MaterialInstanceDynamic)
        (if (or (Utilities|Enum|Equal(Enum) (Resurrection|GetSpawnMachineState) "WaitingForUse") (Utilities|Enum|Equal(Enum) (Resurrection|GetSpawnMachineState) "BeingUsed"))
          (Components|Activation|Activate _activatesound true)
          (SoundIndicator|StartSoundIndicatorLooping (Variables|Default|GetFortSoundIndicatorComponent) (Transformation|GetActorLocation) (Variables|/Script/Engine.AudioComponent|GetSound _activatesound) 1.0 0)
          (CallFunction|StartActivateSoundTimer)))
      (:then_2
        ;[1 0.6214 0.2247] Initiate MultiInteract's screens mat params
        (if (Class|FortActorComponentMultiInteract|IsMultipleInteractionEnabled (Variables|Default|GetMultiInteractActorComponent))
          (LocalVariables|SetMultiInteractWeight 1.0)
          (else
            (LocalVariables|SetMultiInteractWeight 0.0)))
        (Rendering|Material|SetScalarParameterValue (Variables|FX|GetScreensMaterial) "NewMultiWeight" (LocalVariables|GetMultiInteractWeight) :!class /Script/Engine.MaterialInstanceDynamic)))
    (else
      (Rendering|Material|SetScalarParameterValue (Variables|FX|GetScreensMaterial) "InteractTime" 0.0 :!class /Script/Engine.MaterialInstanceDynamic)
      (Effects|Components|ParticleSystem|SetFloatParameter (Variables|Default|GetNS_SCMachine_HoloLogo_LW) "GameTime" 0.0)
      (Audio|Components|Audio|Play (Variables|Default|GetActivateEndSound) 0.0)
      (Audio|Components|Audio|FadeOut (Variables|Default|GetActivateSound) 0.5 0.0 "Linear" :!class /Script/Engine.AudioComponent)
      (SoundIndicator|StopSoundIndicatorLooping (Variables|Default|GetFortSoundIndicatorComponent)))))

(fn ShouldDie
  (params (Damage :type Float%20%28single-precision%29) (EventInstigator :type Controller%20Object%20Reference) (DamageCauser :type Actor%20Object%20Reference) (bIsTest :type Boolean) (EffectContext :type Gameplay%20Effect%20Context%20Handle%20Structure%20%28by%20ref%29))
  (return false))

(fn UpdateScreen
  (bind _returnvalue (Resurrection|GetSpawnMachineState))
  (bind _returnvalue_2 (Utilities|Time|GetGameTimeinSeconds))
  (bind _timer_cooldown (Variables|States|GetTimer_Cooldown))
  (bind _returnvalue_3 (Utilities|Time|GetTimerRemainingTimebyHandle _timer_cooldown))
  (bind _returnvalue_4 (<= _returnvalue_3 0))
  (bind _cooldownlength (LocalVariables|GetCooldownLength))
  (bind _returnvalue_5 (Math|Float|SelectFloat _cooldownlength _returnvalue_3 _returnvalue_4))
  (bind _screensmaterial (Variables|FX|GetScreensMaterial))
  (bind _screensmaterial_2 (Variables|FX|GetScreensMaterial))
  (bind _ns_scmachine_holologo_lw (Variables|Default|GetNS_SCMachine_HoloLogo_LW))
  (Rendering|Material|SetScalarParameterValue _screensmaterial "Spawning" (Math|Conversions|ToFloat(Boolean) (Utilities|Enum|Equal(Enum) _returnvalue "Active")) :!class /Script/Engine.MaterialInstanceDynamic) ; Spawning state
  (Rendering|Material|SetScalarParameterValue _screensmaterial "Cooldown" (Math|Conversions|ToFloat(Boolean) (Utilities|Enum|Equal(Enum) _returnvalue "OnCooldown")) :!class /Script/Engine.MaterialInstanceDynamic) ; Cooldown state
  (switch Utilities|FlowControl|Switch|SwitchonESpawnMachineState (Resurrection|GetSpawnMachineState)
    (:WaitingForUse
      (Fortnite|SetNiagaraAssetMatchActivation (Variables|Default|GetNS_SCMachine_HoloLogo_LW) "/Game/Athena/Items/EnvironmentalItems/SCMachine/FX/NS_SCMachine_HoloLogo_LW.NS_SCMachine_HoloLogo_LW" false))
    (:Active)
    (:Complete
      (Rendering|Material|SetScalarParameterValue (Variables|FX|GetScreensMaterial) "SpawningTime" 0.0 :!class /Script/Engine.MaterialInstanceDynamic))
    (:OnCooldown
      (Fortnite|SetNiagaraAssetMatchActivation (Variables|Default|GetNS_SCMachine_HoloLogo_LW) "/Game/Athena/Items/EnvironmentalItems/SCMachine/FX/NS_SCMachine_HoloBattery_LW.NS_SCMachine_HoloBattery_LW" false)
      (LocalVariables|SetCooldownLength (Resurrection|GetCooldownLength))
      (Utilities|FlowControl|Sequence
        (:then_0
          (Effects|Components|ParticleSystem|SetFloatParameter _ns_scmachine_holologo_lw "GameTime" _returnvalue_2)
          (Effects|Components|ParticleSystem|SetFloatParameter _ns_scmachine_holologo_lw "InteractDuration" _returnvalue_5))
        (:then_1
          (Rendering|Material|SetScalarParameterValue _screensmaterial_2 "CooldownTime" _returnvalue_2 :!class /Script/Engine.MaterialInstanceDynamic)
          (Rendering|Material|SetScalarParameterValue _screensmaterial_2 "CooldownDuration" _returnvalue_5 :!class /Script/Engine.MaterialInstanceDynamic))))))

(fn ConstructionScript
  (bind _self self)
  (bind _returnvalue (Transformation|GetActorLocation _self))
  (bind _x (.x _returnvalue))
  (bind _y (.y _returnvalue))
  (bind _z (.z _returnvalue))
  (bind _returnvalue_2 (Math|Float|float+float _x _y _z))
  (bind _returnvalue_3 (/ _returnvalue_2 3.0))
  (bind _returnvalue_4 (Math|Float|Truncate _returnvalue_3))
  (bind _returnvalue_5 (Utilities|String|ToString(Integer) _returnvalue_4))
  (bind _returnvalue_6 (Utilities|GetDisplayName _self))
  (bind _returnvalue_7 (Utilities|String|Append _returnvalue_6 _returnvalue_5))
  (bind _van (Variables|Default|GetVan))
  (|CallParentFunction)
  (Utilities|IsValid (Variables|/Script/Engine.StaticMeshComponent|GetStaticMesh (Variables|Default|GetVan))
    (:Is%20Valid
      (bind _returnvalue_8 (Rendering|Material|CreateDynamicMaterialInstance _van 2 "/Game/Athena/Items/Gameplay/Second_Chance/Materials/MI_Second_Chance_Screen.MI_Second_Chance_Screen" (Utilities|String|StringToName (Utilities|String|Append _returnvalue_7 "_A")) :!class /Script/Engine.PrimitiveComponent))
      (Variables|FX|SetScreensMaterial _returnvalue_8)
      (bind _returnvalue_9 (Rendering|Material|CreateDynamicMaterialInstance _van 0 0 (Utilities|String|StringToName (Utilities|String|Append _returnvalue_7 "_B")) :!class /Script/Engine.PrimitiveComponent))
      (Variables|FX|SetVanBodyMaterial _returnvalue_9))))
