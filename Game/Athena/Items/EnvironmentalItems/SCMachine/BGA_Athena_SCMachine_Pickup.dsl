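; Reboot card pickup that only the owner's team sees and can collect, by interaction, touch or server auto-loot.
; On BeginPlay it destroys itself in Creative Royale edit areas; otherwise it unhides after a delay and starts a lifetime timer.
; Expiry blocks new interactions, then waits one interact duration before destruction.
; RepNotify flags, set after flushing net dormancy, drive unhide, team collect FX and hide-and-kill with purchase-aware FX.
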(event Custom|SpawnSound
  ;[1 1 1] Spawn Sound - Play the 2d spawn sound. We then need to set this bool so it does not play again. The delay alows time for the sound to happen on the clients before the server sets it to false.
  (if (not (Variables|Audio|GetSpawnSoundPlayed))
    (for _array_element (Fortnite|GetLocalFortPlayerControllers true)
      (if (and (or (Utilities|IsValid (Pawn|GetSpectatorPawn _array_element) :!class /Script/Engine.KismetSystemLibrary) (Utilities|IsValid (Game|Player|GetPlayerPawn _array_element) :!class /Script/Engine.KismetSystemLibrary)) (Math|Byte|Equal(Byte) (Team|GetActorTeam _array_element) (Variables|Default|GetOwnerTeam)))
        (Audio|PlaySound2D (Variables|Audio|GetSpawnInSound))))
    (Utilities|FlowControl|Delay 1.0)
    (CallFunction|MarkSpawnSoundasPlayed)))

(event Custom|DestroyPickup
  ;[1 1 1] Set a Rep Notify Var to trigger the event for the players.
  (if (and (Networking|IsServer) (not (Variables|Default|GetHideAndKill)))
    (CallFunction|EnableHideAndKill)))

(event Custom|DelayDestroyPickup
  ;[1 1 1] Delay of pickup destruction in case the player is interacting it at the very last seconds
  (Variables|Default|SetIsDelayingDeath true)
  (Utilities|FlowControl|Delay (ScalableFloat|GetValueatLevelwithContext (Variables|Default|GetRow_PickUpInteractTime)))
  (CallFunction|DestroyPickup))

(event EventBeginPlay
  ;[1 1 1] Begin Play
  (|CallParentFunction)
  ;;[1 0.9252 0.277] Destroy the reboot card if it spawns in the Creative Royale Edit Plot. This prevents players joining in progress on an island with Join In Progress set to spectate from dropping a reboot card.
  (bind _returnvalue (Class|PlayspaceLibrary|GetPlayspaceforActor self self))
  (Utilities|IsValid _returnvalue
    (:Is%20Valid
      (if (not (Actor|ActorHasTag _returnvalue "CreativeRoyale.Playspace.EditArea"))
        (Utilities|FlowControl|Sequence
          (:then_0
            (CallFunction|SetTeam (Variables|Default|GetOwnerTeam))
            (if (Networking|IsServer)
              ;; Enable overlap events on the server to allow teammates to pick up the card by touching it
              (Variables|/Script/Engine.PrimitiveComponent|SetbGenerateOverlapEvents true (Variables|Default|GetOverlapCollision))))
          (:then_1
            (if (Networking|IsServer)
              (bind _asfort_player_controller_athena (Utilities|Casting|CastToFortPlayerControllerAthena (Variables|Spawn|GetOwnerPlayerController)))
              (bind _returnvalue_2 (Resurrection|GetAutoLootRebootCardTeamMemberController _asfort_player_controller_athena))
              (Utilities|IsValid _returnvalue_2
                (:Is%20Valid
                  (Resurrection|PickUp _returnvalue_2)
                  (CallFunction|MarkasCollected))
                (:Is%20Not%20Valid
                  _)))
            (Utilities|FlowControl|Sequence
              (:then_0
                (Utilities|FlowControl|Delay (Variables|Unhide|GetDelayBeforeUnhide))
                (if (not (Variables|Default|GetIsPendingKill))
                  (CallFunction|SpawnSound)
                  (if (not (Rendering|IsVisible (Variables|Default|GetOverlapCollision)))
                    (CallFunction|SetUnHide 1)
                    ;[1 1 1] Setup Destroy Pickup
                    (bind _returnvalue_3 (Utilities|Time|SetTimerbyEvent (delegate Custom|DelayDestroyPickup) (ScalableFloat|GetValueatLevelwithContext (Variables|Default|GetRow_PickupLife))))
                    (Variables|Default|SetTimer_DestroyPickup _returnvalue_3))))
              (:then_1
                (FortPhysicsSimulationLibrary|WakeUp self "(MaxSimSizeAllowed=Large,bRejectEmbeddedInLandscape=False)")))))
        (else
          (Actor|DestroyActor))))))

(event Custom|CollectPickup
  ;[1 1 1] Setup Destroy Pickup
  (Utilities|FlowControl|Sequence
    (:then_0
      (for _array_element (Fortnite|GetLocalFortPlayerControllers true)
        (if (Math|Byte|Equal(Byte) (Team|GetActorTeam _array_element) (Variables|Default|GetOwnerTeam))
          (Utilities|IsValid (Game|Player|GetPlayerPawn _array_element)
            (:Is%20Valid
              (Niagara|SpawnSystemAttached "/Game/Athena/Items/EnvironmentalItems/SCMachine/FX/NS_SCMachine_CardCollect_Shared.NS_SCMachine_CardCollect_Shared" (Variables|/Script/Engine.Character|GetMesh (Game|Player|GetPlayerPawn _array_element))))))))
    (:then_1
      (CallFunction|DestroyPickup))))

(event Custom|ForceGiveRebootCardForTeamMembers
  (bind _asfort_player_controller_athena (Utilities|Casting|CastToFortPlayerControllerAthena (Variables|Spawn|GetOwnerPlayerController)))
  (Utilities|IsValid _asfort_player_controller_athena
    (:Is%20Valid
      (Resurrection|PickUp _asfort_player_controller_athena)
      (CallFunction|MarkasCollected))))

(event Custom|HideAndKillPickup
  ;[1 1 1] Setup Destroy Pickup
  (bind _returnvalue (Niagara|SpawnSystematLocation "/Game/Athena/Items/EnvironmentalItems/SCMachine/FX/NS_SCCard_Collect.NS_SCCard_Collect" (Transformation|GetWorldLocation (Variables|Default|GetCard))))
  (if (Variables|Default|GetWas%20Card%20Purchased)
    (Niagara|SetNiagaraVariable(Bool) _returnvalue "Purchased" true)
    (else
      (Niagara|SetNiagaraVariable(Bool) _returnvalue "Purchased" false)))
  (Audio|Components|Audio|FadeOut (Variables|Default|GetSC_Machine_Memory_Card_Loop_Cue) 0.15 0.0 "Linear" :!class /Script/Engine.AudioComponent)
  (Audio|PlaySoundatLocation "/Game/Athena/Sounds/SC_Machine/SC_Machine_Card_PickUp_Cue.SC_Machine_Card_PickUp_Cue" (Transformation|GetActorLocation))
  (Collision|SetActorEnableCollision false)
  (Rendering|SetActorHiddenInGame true)
  (Actor|SetLifeSpan (Math|Random|RandomFloatinRange 0.01 0.5)))

(event Interaction|EventBlueprintOnInteract
  (params (InteractingPawn :type Fort%20Pawn%20Object%20Reference) (InteractionBeingAttempted :type EInteractionBeingAttempted%20Enum :default "FirstInteraction"))
  ;[1 1 1] On Interact
  (CallFunction|MarkasCollected))

(event OnLinkedActorDestroyed(FortLinkToActor)
  (params (DamageCauser :type Actor%20Object%20Reference))
  ;;[1 1 1] Base destroyed. Start simulating again.
  ;[0.1942 1 0.1821] Physics
  (FortPhysicsSimulationLibrary|WakeUp self "(MaxSimSizeAllowed=Large,bRejectEmbeddedInLandscape=False)"))

(event Resurrection|EventOnDestroyPickup
  (params (bWasCardPurchased :type Boolean))
  ;[1 1 1] Handle force destroying the pickup
  (Variables|Default|SetWas%20Card%20Purchased bWasCardPurchased)
  (CallFunction|DestroyPickup))

(event EventOnDeathServer
  (params (Damage :type Float%20%28single-precision%29) (DamageTags :type Gameplay%20Tag%20Container%20Structure%20%28by%20ref%29) (Momentum :type Vector :default "0, 0, 0") (HitInfo :type Hit%20Result%20Structure%20%28by%20ref%29) (InstigatedBy :type Controller%20Object%20Reference) (DamageCauser :type Actor%20Object%20Reference) (EffectContext :type Gameplay%20Effect%20Context%20Handle%20Structure))
  ;[1 1 1] Set the is pending kill var when the chip is killed to make sure that if we kill it on disconncet it does not play sounds or visuals.
  (CallFunction|MarkAsPendingKill))

(event OnComponentBeginOverlap(OverlapCollision)
  (params (OverlappedComponent :type Primitive%20Component%20Object%20Reference) (OtherActor :type Actor%20Object%20Reference) (OtherComp :type Primitive%20Component%20Object%20Reference) (OtherBodyIndex :type Integer) (bFromSweep :type Boolean) (SweepResult :type Hit%20Result%20Structure%20%28by%20ref%29))
  ;[1 1 1] Allow teammates to collect the reboot card by touching it
  (if (Networking|IsServer)
    (bind _asfort_player_pawn_athena (Utilities|Casting|CastToFortPlayerPawnAthena OtherActor))
    (if (Utilities|NotEqual(Object) (Variables|Spawn|GetOwnerPlayerController) (Pawn|GetController _asfort_player_pawn_athena))
      (if (Math|Byte|Equal(Byte) (Team|GetActorTeam _asfort_player_pawn_athena) (Variables|Default|GetOwnerTeam))
        (bind _asfort_player_controller_athena (Utilities|Casting|CastToFortPlayerControllerAthena (Pawn|GetController _asfort_player_pawn_athena)))
        (Resurrection|PickUp _asfort_player_controller_athena)
        (CallFunction|MarkasCollected)))))

(fn BlueprintCanInteract
  (params (InteractingPawn :type Fort%20Pawn%20Object%20Reference) (InteractionBeingAttempted :type EInteractionBeingAttempted%20Enum :default "FirstInteraction") (InteractionType :type TInteractionType%20Enum :default "IT_NoInteraction"))
  (return (and (Math|Byte|Equal(Byte) (Team|GetActorTeam InteractingPawn) (Variables|Default|GetOwnerTeam)) (not (Variables|Default|GetIsDelayingDeath)))))

(fn BlueprintGetInteractionString
  (params (InteractingPawn :type Fort%20Pawn%20Object%20Reference) (InteractionBeingAttempted :type EInteractionBeingAttempted%20Enum :default "FirstInteraction"))
  (return (Variables|Default|GetInteractText)))

(fn BlueprintGetInteractionTime
  (params (InteractingPawn :type Fort%20Pawn%20Object%20Reference) (InteractionBeingAttempted :type EInteractionBeingAttempted%20Enum :default "FirstInteraction"))
  ;[0.5412 1 0.3843] There is only 1 type of interaction this actor can have. Check the runtime registry because the value can be changed in Creative Royale.
  (return (ScalableFloat|GetValueatLevelwithContext (Variables|Default|GetRow_PickUpInteractTime)) true))

(fn BuildDataRegistryResolverScope
  (bind _self self)
  (bind _returnvalue (Minigame|GetMinigame _self :!class /Script/FortniteGame.MinigameBlueprintLibrary))
  ;[1 0.8475 0.2776] Gives the name of the volume as a scope to the data registry. Needed for calls to GetValueWithContext to work
  (Utilities|IsValid _returnvalue
    (:Is%20Valid
      (return (Utilities|Array|MakeArray (SpatialGameplay|GetSpatialGameplayInterfaceFNameforActor _returnvalue)) 0 true))
    (:Is%20Not%20Valid
      (return 0 0 false))))

(fn CanHandleCleanupEvents
  (return (not (Math|Boolean|ORBoolean (Variables|Default|GetIsDelayingDeath) (Variables|Default|GetHideAndKill) (Variables|Default|GetIsPendingKill) (Game|IsActorBeingDestroyed)))))

(fn EnableHideandKill
  (Networking|FlushNetDormancy)
  (Variables|Default|SetHideAndKill true))

(fn GetCacheContextOverride
  ;[0.4163 1 0.9801] We don't need to return anything here since we don't need to override the Cache Context.
  (return 0))

(fn MarkAsCollected
  (Networking|FlushNetDormancy)
  (Variables|Default|SetCollectedPickup true))

(fn MarkasPendingKill
  (Networking|FlushNetDormancy)
  (Variables|Default|SetIsPendingKill true))

(fn MarkSpawnSoundAsPlayed
  (Networking|FlushNetDormancy)
  (Variables|Audio|SetSpawnSoundPlayed true))

(fn OnRep_CollectedPickup
  (CallFunction|CollectPickup))

(fn OnRep_HideAndKill
  ;[1 1 1] Call the Hide and Kill Function
  (if (Variables|Default|GetHideAndKill)
    (CallFunction|HideandKillPickup)))

(fn OnRep_UnHide
  (bind _unhide (Variables|Unhide|GetUnHide))
  (bind _returnvalue (Math|Conversions|ToBoolean(Integer) _unhide))
  (Variables|Interaction|SetbAllowInteract true)
  (Development|PrintString "Unhide")
  (Rendering|SetVisibility (Variables|Default|GetOverlapCollision) (Math|Conversions|ToBoolean(Integer) (Variables|Unhide|GetUnHide)) true)
  ;[1 0.008092 0] This is Proto Need to change this for ship to be a function similar to the get local team affiliation
  (for _array_element (Fortnite|GetLocalFortPlayerControllers true)
    (if (Math|Byte|Equal(Byte) (Variables|Default|GetOwnerTeam) (Team|GetActorTeam _array_element))
      ;[1 1 1] Set The indicator color and turn on the glow if the player is on the same team
      (Development|SetHiddeninGame (Variables|Default|GetBackgroundGlow) _returnvalue false)
      (Components|Activation|SetActive (Variables|Default|GetSpawnInEffect) _returnvalue true)
      (Rendering|Material|SetScalarParameterValueonMaterials (Variables|Default|GetCard) "StartTime" (Utilities|Time|GetGameTimeinSeconds))
      (Audio|Components|Audio|FadeIn (Variables|Default|GetSC_Machine_Memory_Card_Loop_Cue) 0.25 1.0 0.0 "Linear" :!class /Script/Engine.AudioComponent)
      (else
        ;[1 1 1] Hide actor if not on same team and turn off its collision
        (Rendering|SetActorHiddenInGame true)
        (Collision|SetActorEnableCollision false)))))

(fn SetUnHide
  (params (UnHide :type Integer))
  (Networking|FlushNetDormancy)
  (Variables|Unhide|SetUnHide UnHide))

(fn ShouldDie
  (params (Damage :type Float%20%28single-precision%29) (EventInstigator :type Controller%20Object%20Reference) (DamageCauser :type Actor%20Object%20Reference) (bIsTest :type Boolean) (EffectContext :type Gameplay%20Effect%20Context%20Handle%20Structure%20%28by%20ref%29))
  (return false))

(fn ConstructionScript
  (Transformation|SetAbsolute (Variables|Default|GetCard) false true false))
