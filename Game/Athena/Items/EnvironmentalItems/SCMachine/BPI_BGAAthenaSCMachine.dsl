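; Blueprint interface for the Athena SC Machine environmental item (a BGA actor).
; It declares a single function, GetHoloLogoNiagaraComponent, which returns the machine's holographic logo Niagara component.
; The interface body holds no logic and returns an empty default, leaving implementers to supply the component.
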(fn GetHoloLogoNiagaraComponent
  (return 0))
