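# Unlit, two-sided translucent material for mesh particles forming the glowing active beacon of the SCMachine item.
# Opacity combines UV gradients, Fresnel, axis-angle, depth and near-camera fades, a sine pulse and time-of-day brightness.
# Emissive blends a color gradient with a hot base and adds a timestamp-driven flash.
# Vertex colors separate spline from body, masking SplineThicken width and height offsets fed by particle dynamic parameters.
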Material /Game/Athena/Items/EnvironmentalItems/SCMachine/FX/M_SCMachine_ActiveBeacon
graph {
 properties {
  bUsedWithMeshParticles=true
  bUsedWithNiagaraMeshParticles=true
  blendMode=BLEND_Translucent
  pixelDepthOffsetMode=PDOM_Legacy
  shadingModel=MSM_Unlit
  twoSided=true
  usedShadingModels=Unlit
 }
 nodes [
  DynamicParameter() {desc=Niagara index 1, parameterIndex=1, defaultValue=<rgba 0 0 0 0>, paramNames [TopWiden,Param2,Param3,Param4]}
  n1 = ParticleColor()
  n2 = ScalarParameter() {defaultValue=0.88, parameterName=MorningEveningBrightness}
  n3 = MaterialFunctionCall(
   Color_Morning=@n2
   Color_Day=Constant() {r=1}
   Color_Evening=@n2
   Color_Night=ScalarParameter() {defaultValue=0.6, parameterName=NightBrightness}
  ) {materialFunction=&Color_Time_of_Day}
  n4 = LinearInterpolate(A=ScalarParameter() {defaultValue=0.25, parameterName=InnerBrightness}, Alpha=ConstantBiasScale(None=TwoSidedSign())) {constA=0.25}
  n5 = ComponentMask(None=TextureCoordinate()) {g=true}
  n6 = ScalarParameter() {defaultValue=0.2, parameterName=FresnelBaseBrightness}
  n7 = VertexColor()
  n8 = LinearInterpolate(
   A=QualitySwitch(Default=Fresnel(BaseReflectFractionIn=@n6) {baseReflectFraction=0.2, exponent=1.5}, Low=Constant() {r=0.3})
   B=ScalarParameter() {defaultValue=0.6, parameterName=SplineBrightness}
   Alpha=@n7.R
  ) {constB=0.75}
  n9 = VectorParameter() {parameterName=AngleFadeAxis, defaultValue=<rgba 0 0 1 1>}
  n10 = ScalarParameter() {defaultValue=1, parameterName=AngleFadeExp}
  n11 = Multiply(A=VertexNormalWS(), B=TwoSidedSign())
  n12 = Multiply(
   A=Power(
    Base=MaterialFunctionCall!(Axis=@n9.RGB, Exponent=@n10, Normal=@n11) {materialFunction=/Game/Effects/Fort_Effects/Materials/_Functions/MF_AxisAlignedFrensnel}.Fresnel
    Exp=ScalarParameter() {defaultValue=3, parameterName=AngleFadePower}
   )
   B=ScalarParameter() {defaultValue=5, parameterName=AngleFadeMult}
  ) {constB=5}
  n13 = VertexColor()
  n14 = Clamp(None=QualitySwitch(Default=LinearInterpolate(A=@n12, Alpha=@n7.G), Low=Constant() {r=1}), Min=@n13.R)
  n15 = Power(Base=@n5, Exp=ScalarParameter() {defaultValue=3, parameterName=GradientPower}) {constExponent=3}
  n16 = LinearInterpolate(
   A=ScalarParameter() {defaultValue=40, parameterName=MaxDepthFade}
   B=Multiply(A=ScalarParameter() {defaultValue=150, parameterName=SplineDepthFade}, B=@n13.R) {constB=50}
   Alpha=@n15
  ) {constA=40, constB=0}
  n17 = ScalarParameter() {defaultValue=200, parameterName=NearFadeDistance}
  n18 = Multiply(
   A=@n8
   B=Multiply(
    A=DepthFade(Opacity=@n14, FadeDistance=@n16)
    B=MaterialFunctionCall(Distance=@n17) {materialFunction=&Near_Camera_Fade}.Result
   )
  )
  n19 = MaterialFunctionCall(Speed=Constant() {r=0.5}) {materialFunction=&DebugTimeSine}
  n20 = Multiply(
   A=Multiply(A=Multiply(A=@n5, B=@n18), B=LinearInterpolate(Alpha=@n19.Result) {constA=0.4})
   B=ScalarParameter() {defaultValue=0.64, parameterName=OpacityMult}
  )
  n21 = ComponentMask(None=TextureCoordinate()) {r=true}
  n22 = Constant() {r=0.5}
  n23 = VertexColor()
  n24 = OneMinus(None=@n23.R)
  n25 = Power(Base=Multiply(A=SphereMask(A=@n21, B=@n22, Radius=@n22, Hardness=@n24), B=@n23.B)) {constExponent=1.4}
  n26 = Multiply(A=@n1.A, B=Multiply(A=Multiply(A=@n3.Result, B=@n4), B=Multiply(A=@n20, B=@n25)) {constA=1})
  n27 = Multiply(A=@n26, B=DepthFade(Opacity=Constant() {r=1}, FadeDistance=ScalarParameter() {parameterName=FinalPass-DepthFade-Distance}))
  n28 = ShadingPathSwitch(Default=@n27, Deferred=@n27, Mobile=@n26) {desc=Migrated from FeatureLevelSwitch}
  n29 = Fresnel(ExponentIn=ScalarParameter() {defaultValue=1, parameterName=EdgeFresnelExp}) {baseReflectFraction=0.2, exponent=1.5}
  LinearInterpolate(A=@n28, B=Multiply(A=@n28, B=OneMinus(None=@n29)), Alpha=ScalarParameter() {parameterName=UseEdgeFresnel})
  n30 = Divide(
   A=Subtract(A=Time() {period=5}, B=ScalarParameter() {parameterName=FlashStartTime})
   B=ScalarParameter() {defaultValue=0.75, group=Timestamp, parameterName=FlashDuration}
  )
  n31 = Multiply(
   A=MaterialFunctionCall!(In=Clamp(None=@n30)) {materialFunction=/Game/Effects/Fort_Effects/Materials/_Functions/MF_Smoothstep_EaseOut}.Result
   B=ScalarParameter() {defaultValue=4, parameterName=FlashBrightness}
  )
  n32 = LinearInterpolate(
   A=VectorParameter() {parameterName=TopColor, defaultValue=<rgba 0.2125 0.4052 2 0>}.RGB
   B=VectorParameter() {parameterName=BrightColor, defaultValue=<rgba 0.1497 0.8948 2 0>}.RGB
   Alpha=Multiply(A=@n8, B=@n15)
  )
  n33 = Power(Base=Multiply(A=OneMinus(None=@n7.R), B=@n5), Exp=ScalarParameter() {defaultValue=15, parameterName=HotBasePower}) {constExponent=12}
  n34 = VectorParameter() {parameterName=HotBaseColor, defaultValue=<rgba 0.3906 3.4901 5 0>}
  n35 = LinearInterpolate(
   A=Add(A=@n32, B=Multiply(A=@n33, B=@n34.RGB))
   B=Multiply(A=@n32, B=OneMinus(None=@n33))
   Alpha=ScalarParameter() {parameterName=FadeOutBottom}
  )
  MaterialFunctionCall(Emissive=Multiply(A=Add(A=@n31), B=Multiply(A=@n1.RGB, B=@n35))) {materialFunction=&MF_FX_EmissiveAdjustments}
  n36 = Subtract(A=WorldPosition().XYZ, B=ParticlePositionWS())
  n37 = Normalize(VectorInput=Transform(None=Constant3Vector() {constant=<rgba 1 1 0 0>}) {transformSourceType=TRANSFORMSOURCE_Local})
  n38 = TextureCoordinate()
  Multiply(A=Multiply(A=Multiply(A=@n36, B=@n37)), B=OneMinus(None=ComponentMask(None=@n38) {g=true}))
  n39 = DynamicParameter() {paramNames [HeightAdd,Width,Param3,Param4]}
  n40 = Multiply(A=@n39.Width, B=ScalarParameter() {defaultValue=120, parameterName=SplineWidth})
  n41 = Multiply(
   A=@n23.R
   B=MaterialFunctionCall(WidthBase=@n40, WidthTip=@n40) {materialFunction=&SplineThicken}.WorldPositionOffset
  )
  n42 = Multiply(A=@n23.G, B=Multiply(A=@n39.HeightAdd, B=Constant3Vector() {constant=<rgba 0 0 1 0>}))
  n43 = OneMinus(None=@n23.G)
  Reroute(None=Add(A=Add(A=@n41, B=@n42), B=Multiply(A=@n43)))
 ]
}
