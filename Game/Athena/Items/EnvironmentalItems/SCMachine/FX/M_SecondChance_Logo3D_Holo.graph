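# Translucent unlit hologram material for the Second Chance machine's 3D reboot logo, with a static Battery switch for a battery-icon variant.
# Dynamic parameters for spin, game time and interact duration drive a progress value that fills the logo radially, or fills the blinking battery.
# World position offset bobs the mesh, spinning the logo and bouncing a vertex-color-masked pip upon completion.
# Scanlines, fresnel and camera-distance fades shape emissive and opacity, with cheaper mobile paths.
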Material /Game/Athena/Items/EnvironmentalItems/SCMachine/FX/M_SecondChance_Logo3D_Holo
graph {
 properties {
  bUsedWithMeshParticles=true
  bUsedWithNiagaraMeshParticles=true
  blendMode=BLEND_Translucent
  pixelDepthOffsetMode=PDOM_Legacy
  refractionMethod=RM_IndexOfRefraction
  shadingModel=MSM_Unlit
  usedShadingModels=Unlit
 }
 nodes [
  n1 = Constant()
  n2 = ScalarParameter() {defaultValue=2, parameterName=NumberOfSpins}
  n3 = Add(A=ScalarParameter() {defaultValue=0.25, parameterName=SpinOffset})
  n4 = DynamicParameter() {defaultValue=<rgba 1 1 0 4>, paramNames [Spin, Param2, GameTime, InteractDuration]}
  n5 = Time()
  n6 = Clamp(None=Divide(A=Subtract(A=@n5, B=@n4.GameTime), B=@n4.InteractDuration))
  n7 = Floor(None=@n6)
  n8 = Multiply(A=@n4.Spin, B=@n7)
  n9 = MaterialFunctionCall(Alpha=Clamp(None=Multiply(A=@n3, B=@n8) {constB=1.2})) {materialFunction=&SmoothStep}
  n10 = Transform(None=Constant3Vector() {constant=<rgba 0 0 1 0>}) {transformSourceType=TRANSFORMSOURCE_Local}
  n11 = VertexColor()
  n12 = Multiply(A=ConstantBiasScale(None=Sine(None=@n8)), B=ScalarParameter() {defaultValue=0.3, parameterName=WobbleMult}) {constB=0.1}
  n13 = Power(Base=@n8, Exp=ScalarParameter() {defaultValue=0.6, parameterName=SineAnimPower})
  n14 = ScalarParameter() {defaultValue=0.25, parameterName=SineAnimOffset}
  n15 = Subtract(A=Sine(None=Add(A=@n13, B=@n14) {constB=0.125})) {constB=0.9}
  n16 = OneMinus(None=Clamp(None=Multiply(A=@n15) {constB=12}))
  n17 = Multiply(
   A=Subtract(A=Add(A=@n12, B=@n16), B=ScalarParameter() {defaultValue=0.125, parameterName=PipHeightOffset})
   B=ScalarParameter() {defaultValue=80, parameterName=PipMaxHeight}
  )
  n18 = Constant3Vector() {constant=<rgba 0 0 1 0>}
  n19 = LinearInterpolate(
   A=RotateAboutAxis(
    NormalizedRotationAxis=Normalize(VectorInput=@n10)
    RotationAngle=Multiply(A=@n2, B=@n9.Result) {constB=2}
    PivotPoint=MaterialFunctionCall() {materialFunction=&ObjectPivotPoint}.Object Pivot Location
    Position=WorldPosition().XYZ
   )
   B=Multiply(A=@n11.B, B=Multiply(A=@n17, B=@n18))
   Alpha=@n11.G
  )
  n20 = Time()
  Add(
   A=StaticSwitchParameter(True=@n1, False=ShadingPathSwitch(Default=@n19, Mobile=@n1) {desc=Migrated from FeatureLevelSwitch}) {parameterName=Battery}
   B=Multiply(A=Sine(None=@n20) {period=2.3}, B=VectorParameter() {parameterName=BobWPO, defaultValue=<rgba 0 0 4 1>}.RGB)
  )
  Constant() {r=1}
  n21 = ParticleColor()
  n22 = Constant()
  n23 = VertexColor()
  n24 = Add(A=ScalarParameter() {defaultValue=0.32, parameterName=SineAnimOffsetEmissive}, B=@n8)
  n25 = Subtract(A=Multiply(A=Sine(None=@n24)) {constB=1.07})
  n26 = Multiply(A=@n23.G, B=Multiply(A=Clamp(None=@n25) {maxDefault=2}) {constB=8})
  n27 = OneMinus(None=TextureSampleParameter2D() {parameterName=Mask, texture=&/T_Reboot_Logo_FX}.B)
  n28 = Add(A=Multiply(A=@n27) {constB=0.5}) {constB=0.1}
  n29 = Distance(A=ObjectPositionWS(), B=CameraPositionWS())
  n30 = ScalarParameter() {defaultValue=-3000, parameterName=DistanceFadeoffset}
  n31 = Saturate(None=Divide(A=Add(A=@n29, B=@n30), B=ScalarParameter() {defaultValue=4000, parameterName=DistanceFade}))
  n32 = Add(A=@n31)
  n33 = Fresnel() {baseReflectFraction=0.02, exponent=3}
  n34 = Multiply(A=@n31) {constB=0.1}
  n35 = ScalarParameter() {defaultValue=0.5, parameterName=CooldownEndDuration}
  n36 = Multiply(A=@n6, B=Add(A=Constant() {r=1}, B=Divide(A=@n35, B=@n4.InteractDuration)) {desc=Duration to show filled battery})
  n37 = Floor(None=@n36)
  n38 = Saturate(None=@n37)
  n39 = VertexColor()
  n40 = VertexColor()
  n41 = Multiply(A=OneMinus(None=@n7), B=@n6) {desc=Go dark when full}
  n42 = Subtract(A=MaterialFunctionCall(UVs=TextureCoordinate(), "Rotation Angle (0-1)"=ScalarParameter() {defaultValue=0.2, parameterName=Origin_0-1}) {materialFunction=&CustomRotator}.Rotated Values) {constB=0.5}
  n43 = Add(
   A=@n41
   B=OneMinus(None=MaterialFunctionCall(Vector or UVs=@n42) {materialFunction=&VectorToRadialValue}.Vector Converted to Angle)
  )
  n44 = Multiply(A=OneMinus(None=@n40.G), B=Clamp(None=Floor(None=@n43)))
  n45 = Divide(A=Constant() {r=0.5}, B=@n4.InteractDuration)
  n46 = Multiply(A=@n40.G, B=Floor(None=Add(A=@n41, B=@n45) {constB=0.02})) {desc=Fill up pip near end}
  n47 = Add(
   A=Clamp(None=Add(A=@n33, B=@n34))
   B=StaticSwitchParameter(True=Multiply(A=@n38, B=@n39.B), False=Add(A=@n44, B=@n46)) {parameterName=Battery}
  )
  n48 = VectorParameter() {parameterName=ScanlineColor, defaultValue=<rgba 0.0562 0.1031 0.15 1>}
  n49 = ScreenPosition()
  n50 = ScalarParameter() {defaultValue=12, parameterName=PanLineTile}
  n51 = TextureSample(UVs=Panner(Coordinate=Multiply(A=@n49.ViewportUV, B=@n50) {constB=4}, Time=Time()) {speedX=0.1, speedY=0.25}) {texture=/Game/Effects/Fort_Effects/Textures/Tile/T_ScanLine_Analog}
  n52 = Add(
   A=Add(A=StaticSwitchParameter(True=@n22, False=@n26) {parameterName=Battery}, B=LinearInterpolate(A=@n28, B=@n32, Alpha=@n47) {constA=0.1})
   B=QualitySwitch(Default=Multiply(A=@n48.RGB, B=@n51.R), Low=Constant())
  )
  n53 = VectorParameter() {parameterName=Color, defaultValue=<rgba 0.1912 0.7734 1.5 1>}
  MaterialFunctionCall(Emissive=Multiply(A=@n21.RGB, B=Multiply(A=@n52, B=@n53.RGB))) {materialFunction=&MF_FX_EmissiveAdjustments}
  n54 = Multiply(A=ComponentMask(None=ScreenPosition().ViewportUV) {g=true}, B=ScalarParameter() {defaultValue=200, parameterName=ScanlineTile})
  n55 = Power(Base=ConstantBiasScale(None=Sine(None=@n54)))
  n56 = MaterialFunctionCall(Distance=ScalarParameter() {defaultValue=1000, parameterName=FadeDistance}) {materialFunction=&Near_Camera_Fade}
  n57 = Multiply(A=@n33) {constB=2}
  n58 = ScalarParameter() {defaultValue=0.4, parameterName=ScanlineAdd}
  n59 = ShadingPathSwitch(
   Default=Add(A=LinearInterpolate(A=@n55, Alpha=@n56.Result) {constA=0.4, constB=0.3}, B=Add(A=@n57, B=@n58)) {constB=0.2}
   Mobile=Constant() {r=0.8}
  ) {desc=Migrated from FeatureLevelSwitch}
  n60 = Multiply(A=@n21.A, B=Clamp(None=LinearInterpolate(A=@n59, Alpha=@n31) {constB=2}, Min=Multiply(A=@n27) {constB=0.1}, Max=@n23.R))
  n61 = Sine(None=Time()) {desc=Blink battery icon, period=0.5}
  n62 = Clamp(None=Round(None=Add(A=@n61)), Min=@n37)
  n63 = ScalarParameter() {defaultValue=1.25, parameterName=BatteryGradMax}
  n64 = ScalarParameter() {defaultValue=-0.325, parameterName=BatteryGradMin}
  n65 = ComponentMask(None=TextureCoordinate()) {r=true}
  n66 = Clamp(None=Add(A=@n36, B=LinearInterpolate(A=@n63, B=@n64, Alpha=@n65) {constA=1.25, constB=-0.75}))
  n67 = LinearInterpolate(A=LinearInterpolate(Alpha=@n39.B) {constA=1, constB=0.3}, Alpha=Multiply(A=Multiply(A=@n62, B=@n39.B), B=Floor(None=@n66))) {constA=0.3}
  n68 = MaterialFunctionCall(Distance=ScalarParameter() {defaultValue=500, parameterName=NearCamFadeDistance}) {materialFunction=&Near_Camera_Fade}
  Multiply(
   A=StaticSwitchParameter(True=Multiply(A=@n67, B=@n60), False=@n60) {parameterName=Battery}
   B=ShadingPathSwitch(Default=Power(Base=@n68.Result) {constExponent=3}, Mobile=Constant() {r=1}) {desc=Migrated from FeatureLevelSwitch}
  )
  Time() {bOverride_Period=true, period=4}
 ]
}
