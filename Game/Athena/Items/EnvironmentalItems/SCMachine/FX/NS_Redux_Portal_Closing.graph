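# Niagara effect for the SCMachine portal closing, a roughly ten-second buildup that ends in a collapse.
# A rift mesh swells then snaps down, alongside crack meshes, floating fragments, electric arcs and a flickering corona with a light.
# GPU embers spawn on a surrounding shell and are pulled inward by strengthening attraction, while a delayed shockwave mesh shrinks and fades.
# System.Scale_Closing scales most layers, and age-based lifetimes make staggered cracks and fragments expire near-together.
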NiagaraSystem /Game/Athena/Items/EnvironmentalItems/SCMachine/FX/NS_Redux_Portal_Closing
dependencies=null
emitters {
 Corona {
  emitterUpdateScript [
   &EmitterState(Life Cycle Mode=::NewEnumerator1, Loop Behavior=::NewEnumerator1, Loop Duration=1, UseLoopDelay=true)
   &SpawnBurst_Instantaneous(Spawn Count=1, Use Spawn Probability=false)
  ]
  particleSpawnScript [
   &InitializeParticle(
    Lifetime=10
    Mesh Scale Mode=::NewEnumerator3
    Sprite Size Mode=::NewEnumerator3
    Uniform Sprite Size=150
    UsePositionOffset=false
    Write Material Random=true
   )
   /SetVariables(Particles.LightExponent=3) {isSetParameters=true}
  ]
  particleUpdateScript [
   &ParticleState()
   &ScaleColor(
    Scale Alpha=&FloatFromCurve(FloatCurve=DataInterface() {propertyValues {Curve {keys [<at 0 0 RCIM_Cubic>, {arriveTangent=1, interpMode=RCIM_Cubic, leaveTangent=1, time=0.2355, value=0.501}, <at 1 1>]}}})
    Scale RGB=<1.1458 2.1149 4>
    ScaleA=true
    ScaleRGB=true
   )
   &DynamicMaterialParameters(
    Index 0 Param 1=&FloatFromCurve(FloatCurve=DataInterface() {propertyValues {Curve {keys [<at 0 1.5 RCIM_Cubic>, {interpMode=RCIM_Cubic, leaveTangent=-2.4626, time=1, value=0.3}]}}})
    Index 0 Param 2=0.4
    Index 0 Param 3=1
    Index 0 Param 4=1
    Module.Param0WriteEnabled=true
    Module.Param1WriteEnabled=true
    Module.Param2WriteEnabled=true
    Module.Param3WriteEnabled=true
   )
   &ScaleSpriteSize(Scale Factor=&Vector2DFromFloat(Value=&FloatFromCurve(FloatCurve=DataInterface() {propertyValues {Curve {keys [<at 0 1 RCIM_Cubic>, {interpMode=RCIM_Cubic, leaveTangent=-2.0936, time=1, value=0}]}}})))
   /SetVariables(
    Particles.Color_SharpFlare:LinearColor=&MakeLinearColorFromVectorAndFloat(
     "Vector (RGB)"=&VectorFromFloat(Value=&FloatFromCurve(FloatCurve=DataInterface() {propertyValues {Curve {keys [<at 0 1 RCIM_Cubic>, {interpMode=RCIM_Cubic, leaveTangent=12.6773, time=1, value=5}]}}}))
     "Float (Alpha)"=1
    )
    Particles.SpriteSize_SharpFlare:Vector2f=&Multiply_Vector2D_ByFloat(Float=$System.Scale_Closing, Vector2D=<210 66>)
   ) {isSetParameters=true}
   /SetVariables(
    Particles.Color_Glow:LinearColor=&MakeLinearColorFromVectorAndFloat(
     "Vector (RGB)"=<0.3323 1.6388 2>
     "Float (Alpha)"=&FloatFromCurve(FloatCurve=DataInterface() {propertyValues {Curve {keys [
          <at 0.2 0 RCIM_Cubic>
          {arriveTangent=0.3612, interpMode=RCIM_Cubic, leaveTangent=0.3612, time=0.5935, value=0.248}
          {arriveTangent=4.259, interpMode=RCIM_Cubic, leaveTangent=4.259, time=0.6033, value=0.1457}
          {arriveTangent=-3.3878, interpMode=RCIM_Cubic, leaveTangent=-3.3878, time=0.6186, value=0.3549}
          {arriveTangent=3.675, interpMode=RCIM_Cubic, leaveTangent=3.675, time=0.6299, value=0.0554}
          {arriveTangent=1.0154, interpMode=RCIM_Cubic, leaveTangent=1.0154, time=0.649, value=0.4666}
          {arriveTangent=4.9096, interpMode=RCIM_Cubic, leaveTangent=4.9096, time=0.6737, value=0.0999}
          {arriveTangent=4.8188, interpMode=RCIM_Cubic, leaveTangent=4.8188, time=0.6891, value=0.6636}
          {arriveTangent=-7.7413, interpMode=RCIM_Cubic, leaveTangent=-7.7413, time=0.7042, value=0.2469}
          {arriveTangent=1.3867, interpMode=RCIM_Cubic, leaveTangent=1.3867, time=0.7167, value=0.4498}
          {arriveTangent=4.8476, interpMode=RCIM_Cubic, leaveTangent=4.8476, time=0.7514, value=0.3124}
          {arriveTangent=0.4801, interpMode=RCIM_Cubic, leaveTangent=0.4801, time=0.7584, value=0.6518}
          {arriveTangent=2.0481, interpMode=RCIM_Cubic, leaveTangent=2.0481, time=0.779, value=0.3256}
          {arriveTangent=1.5117, interpMode=RCIM_Cubic, leaveTangent=1.5117, time=0.791, value=0.7187}
          {arriveTangent=6.4492, interpMode=RCIM_Cubic, leaveTangent=6.4492, time=0.8152, value=0.3803}
          {arriveTangent=2.8686, interpMode=RCIM_Cubic, leaveTangent=2.8686, time=0.8271, value=0.9515}
          {arriveTangent=-5.7582, interpMode=RCIM_Cubic, leaveTangent=-5.7582, time=0.852, value=0.4859}
          {arriveTangent=11.9335, interpMode=RCIM_Cubic, leaveTangent=11.9335, time=0.8932, value=0.5711}
          {arriveTangent=0.2435, interpMode=RCIM_Cubic, leaveTangent=0.2435, time=0.8965, value=1.017}
          {arriveTangent=-8.9415, interpMode=RCIM_Cubic, leaveTangent=-8.9415, time=0.9148, value=0.5764}
          {arriveTangent=19.6297, interpMode=RCIM_Cubic, leaveTangent=19.6297, time=0.94, value=0.6286}
          {arriveTangent=-2.6018, interpMode=RCIM_Cubic, leaveTangent=-2.6018, time=0.9435, value=1.1414}
          {arriveTangent=-20.9801, interpMode=RCIM_Cubic, leaveTangent=-20.9801, time=0.9583, value=0.5809}
          {arriveTangent=39.3051, interpMode=RCIM_Cubic, leaveTangent=39.3051, time=0.969, value=0.6074}
          {arriveTangent=3.3966, interpMode=RCIM_Cubic, leaveTangent=3.3966, time=0.9742, value=1.2074}
          {arriveTangent=-8.0532, interpMode=RCIM_Cubic, leaveTangent=-8.0532, time=0.9893, value=0.6763}
          <at 1 1>
         ]}}})
    )
    Particles.SpriteSize_Glow:Vector2f=&Multiply_Vector2D_ByFloat(Float=$System.Scale_Closing, Vector2D=<140 140>)
   ) {isSetParameters=true}
   &CameraOffset(Camera Offset Amount=25)
   /SetVariables(Particles.LightRadius=&Multiply_Float(A=400, B=$System.Scale_Closing)) {isSetParameters=true}
   /SetVariables(Particles.LightColor=&MultiplyLinearColor(
     Linear Color=&MakeLinearColorFromVectorAndFloat(
      "Vector (RGB)"=<1 1 1>
      "Float (Alpha)"=&FloatFromCurve(
       FloatCurve=DataInterface() {propertyValues {Curve {keys [
           <at 0 0 RCIM_Cubic>
           {arriveTangent=4.1026, interpMode=RCIM_Cubic, leaveTangent=0.6001, tangentMode=RCTM_Break, time=0.05, value=0.5}
           {arriveTangent=11.2956, interpMode=RCIM_Cubic, leaveTangent=-75.0977, tangentMode=RCTM_Break, time=0.96, value=4}
           <at 1 0>
          ]}}}
       Scale Curve=10
      )
     )
     Scale Factor=&MakeVector4_FromLinearColor(LinearColor=<rgba 0.147 0.5957 1 1>)
    )) {isSetParameters=true}
  ]
  rendererClasses [&NiagaraSpriteRendererProperties, &NiagaraLightRendererProperties]
  renderers [
   {rendererClass=&NiagaraSpriteRendererProperties}
   {rendererClass=&NiagaraSpriteRendererProperties, rendererIndex=1}
   {rendererClass=&NiagaraSpriteRendererProperties, rendererIndex=2}
   {rendererClass=&NiagaraLightRendererProperties, rendererIndex=3}
  ]
 }
 CrackSpline_ {
  emitterSpawnScript [/SetVariables(Emitter.CrackWidth=60) {isSetParameters=true}]
  emitterUpdateScript [
   &EmitterState()
   &SpawnBurst_Instantaneous(Spawn Count=1, Use Spawn Probability=false)
   &SpawnBurst_Instantaneous(Spawn Count=1, Spawn Group=1, Spawn Time=0.12, Use Spawn Probability=false) {module=SpawnBurst_Instantaneous001}
  ]
  particleSpawnScript [
   /Niagara/Modules/Spawn/Initialization/InitializeParticle(
    Lifetime=&FloatFromCurve(FloatCurve=DataInterface() {propertyValues {Curve {keys [<at 0 10.25>,<at 2 6.25>]}}}, CurveIndex=$System.Age)
    Mesh Scale=&VectorFromFloat(Value=&RandomRangeFloat(Minimum=1.8, Maximum=2.1, Evaluation Type=::NewEnumerator1))
    Module.Write Color=true
    Module.Write Lifetime=true
    Module.Write Mass=true
    Module.Write Material Random=true
    Module.Write Position=false
    Module.Write Scale=true
    Module.Write SpriteRotation=false
    Module.Write SpriteSize=false
   )
   /SetVariables(Particles.MeshOrientation=&MakeQuatFromAxisAngle(
     Axis=<1 1 1>
     AngleInDegrees=&MaskFloatBySpawnGroup(
      Masked Float=&RandomRangeFloat(Minimum=-80, Maximum=0)
      Passthrough Float=&RandomRangeFloat(Minimum=80, Maximum=180)
     )
    )) {isSetParameters=true}
   &Color(Color=&MakeLinearColorFromVectorAndFloat("Vector (RGB)"=<1.5 1.5 1.5>, "Float (Alpha)"=&ClampFloat(Float=1)))
  ]
  particleUpdateScript [
   &ParticleState()
   &ScaleColor(
    Scale RGB=&VectorFromFloat(Value=&FloatFromCurve(FloatCurve=DataInterface() {propertyValues {Curve {keys []}}}))
    ScaleA=true
    ScaleRGB=true
   ) {enabled=false}
   &DynamicMaterialParameters(
    Index 0 Param 1=&FloatFromCurve(FloatCurve=DataInterface() {propertyValues {Curve {keys [
         {arriveTangent=0.249, time=0, value=0}
         {arriveTangent=0.0112, leaveTangent=0.0112, time=0.0429, value=0.249}
         {arriveTangent=0.015, leaveTangent=0.015, time=0.0478, value=0.2602}
         {arriveTangent=0.0229, leaveTangent=0.0229, time=0.161, value=0.2752}
         {arriveTangent=0.0047, leaveTangent=0.0047, time=0.1647, value=0.298}
         {arriveTangent=0.0502, leaveTangent=0.0502, time=0.2839, value=0.3027}
         {arriveTangent=-0.0006, leaveTangent=-0.0006, time=0.2998, value=0.3529}
         {arriveTangent=0.1459, leaveTangent=0.1459, time=0.4532, value=0.3524}
         {arriveTangent=0.0682, leaveTangent=0.0682, time=0.4728, value=0.4983}
         {arriveTangent=0.219, leaveTangent=0.219, time=0.6953, value=0.5665}
         {arriveTangent=2.1176, interpMode=RCIM_Cubic, leaveTangent=2.1176, time=0.7211, value=0.7854}
         {arriveTangent=-2.8157, interpMode=RCIM_Cubic, leaveTangent=-2.8157, time=0.9, value=1}
         <at 1 0 RCIM_Cubic>
        ]}}})
    Index 0 Param 2=&FloatFromCurve(FloatCurve=DataInterface() {propertyValues {Curve {keys [
         <at 0.4 0.4 RCIM_Cubic>
         {arriveTangent=-0.5006, interpMode=RCIM_Cubic, leaveTangent=-0.5006, time=0.95, value=0.7}
         <at 0.9993 0.1 RCIM_Cubic>
        ]}}})
    Index 0 Param 3=1
    Index 0 Param 4=&FloatFromCurve(FloatCurve=DataInterface() {propertyValues {Curve {keys [<at 0 1 RCIM_Cubic>,<at 1 1>]}}})
    Module.Param0WriteEnabled=true
    Module.Param1WriteEnabled=true
    Module.Param2WriteEnabled=true
    Module.Param3WriteEnabled=true
   )
   &CameraOffset(Camera Offset Amount=30)
   &ScaleMeshSize(Scale Factor=&VectorFromFloat(Value=$System.Scale_Closing))
  ]
  rendererClasses [&NiagaraMeshRendererProperties]
  renderers [{rendererClass=&NiagaraMeshRendererProperties, rendererIndex=0}]
 }
 ElectricArcsResidual_ {
  emitterUpdateScript [
   &EmitterState(
    Enable Distance Culling=true
    Life Cycle Mode=::NewEnumerator1
    Loop Behavior=::NewEnumerator1
    Loop Delay=7
    Loop Duration=3.1
    MaxDistance=2000
    Module.bUseMaxDistance=true
    Module.bUseMinDistance=false
    Scalability Mode=::NewEnumerator1
    UseLoopDelay=true
   )
   &SpawnRate(SpawnRate=5, Use Spawn Probability=false)
   &SpawnBurst_Instantaneous(Spawn Count=5, Spawn Time=2.2, Use Spawn Probability=false)
   &SpawnBurst_Instantaneous(Spawn Count=6, Spawn Time=3, Use Spawn Probability=false) {module=SpawnBurst_Instantaneous001}
   &SpawnBurst_Instantaneous(Spawn Count=4, Spawn Time=3.05, Use Spawn Probability=false) {module=SpawnBurst_Instantaneous002}
  ]
  particleSpawnScript [
   /SetVariables(Particles.MeshOrientation=&MakeQuatFromAxisAngle(
     Axis=&RandomRangeVector(Minimum=<0 0 0>, Maximum=<1 0 0>, Evaluation Type=::NewEnumerator1)
     AngleInDegrees=&RandomRangeFloat(Minimum=0, Maximum=360)
    )) {isSetParameters=true}
   &InitializeParticle(
    Lifetime=&RandomRangeFloat(Minimum=0.08, Maximum=0.2, Evaluation Type=::NewEnumerator1)
    Mesh Scale=&RandomRangeVector(Minimum=<2.4 1.8 1.8>, Maximum=<4.5 6 6>, Evaluation Type=::NewEnumerator1)
    Module.Write Color=true
    Module.Write Lifetime=true
    Module.Write Mass=true
    Module.Write Material Random=true
    Module.Write Position=false
    Module.Write Scale=true
    Module.Write SpriteRotation=false
    Module.Write SpriteSize=false
   )
   &DynamicMaterialParameters(
    Index 0 Param 1=&RandomRangeFloat(Minimum=2, Maximum=3, Evaluation Type=::NewEnumerator1)
    Index 0 Param 3=&RandomRangeFloat(Minimum=0.2, Maximum=2, Evaluation Type=::NewEnumerator1)
    Index 0 Param 4=&RandomRangeFloat(Minimum=0, Maximum=1, Evaluation Type=::NewEnumerator1)
    Module.Param0WriteEnabled=true
    Module.Param1WriteEnabled=true
    Module.Param2WriteEnabled=true
    Module.Param3WriteEnabled=true
   )
   &Color(Color=&MakeLinearColorFromVectorAndFloat("Vector (RGB)"=<0.0986 1 0.6632>, "Float (Alpha)"=&ClampFloat(Float=2)))
   /Niagara/Modules/Spawn/Location/SphereLocation(MaskSpawnGroup=false, Module.Override Randomness=false, Sphere Radius=3, Surface Only=true)
   &AddVelocityFromPoint(Module.Use Falloff Distance=false)
  ]
  particleUpdateScript [
   &ParticleState()
   &DynamicMaterialParameters(
    Index 0 Param 2=&FloatFromCurve(FloatCurve=DataInterface() {propertyValues {Curve {keys [{arriveTangent=-1.8894, interpMode=RCIM_Cubic, time=0, value=1}, <at 1 -0.25>]}}})
    Module.Param0WriteEnabled=true
    Module.Param1WriteEnabled=true
    Module.Param2WriteEnabled=true
    Module.Param3WriteEnabled=true
   ) {module=DynamicMaterialParameters001}
   &ScaleColor(
    Scale RGB=&VectorFromCurve(VectorCurve=DataInterface() {propertyValues {
       XCurve {keys [<at 0 50 RCIM_Cubic>, {arriveTangent=-45, interpMode=RCIM_Cubic, leaveTangent=-45, time=0.5, value=30}, <at 1 5>]}
       YCurve {keys [<at 0 19.1123 RCIM_Cubic>, {arriveTangent=-18.2463, interpMode=RCIM_Cubic, leaveTangent=-18.2463, time=0.5, value=13.6826}, <at 1 0.8661>]}
       ZCurve {keys [<at 0 22.1016 RCIM_Cubic>, {arriveTangent=-21.4843, interpMode=RCIM_Cubic, leaveTangent=-21.4843, time=0.5, value=12.787}, <at 1 0.6173>]}
      }})
    ScaleA=true
    ScaleRGB=true
   )
   &ScaleMeshSize(Scale Factor=&VectorFromCurve(
     VectorCurve=DataInterface() {propertyValues {
       XCurve {keys [<at 0.0033 0.5 RCIM_Constant>,<at 0.2054 0.7619 RCIM_Constant>,<at 0.4101 0.8575 RCIM_Constant>,<at 0.6967 1>]}
       YCurve {keys [<at 0.0033 0.6004 RCIM_Constant>,<at 0.2054 0.6716 RCIM_Constant>,<at 0.4101 0.3646 RCIM_Constant>,<at 0.6967 0.2468>]}
       ZCurve {keys [<at 0.0033 0.7767 RCIM_Constant>,<at 0.2054 0.5215 RCIM_Constant>,<at 0.4101 0.4457 RCIM_Constant>,<at 0.6967 0.6282>]}
      }}
     Scale Curve=&VectorFromFloat(Value=$System.Scale_Closing)
    ))
   &ScaleMeshSize(Scale Factor=&VectorFromFloat(Value=&FloatFromCurve(FloatCurve=DataInterface() {propertyValues {Curve {keys [<at 0 1>,<at 15000 1.785>]}}}, CurveIndex=$Engine.Owner.LODDistance))) {module=ScaleMeshSize001}
   &CameraOffset(Camera Offset Amount=30)
   &SolveForcesAndVelocity(Clamp Velocity=false, Limit Acceleration=false)
  ]
  rendererClasses [&NiagaraMeshRendererProperties]
  renderers [{rendererClass=&NiagaraMeshRendererProperties, rendererIndex=0}]
 }
 Embers_ {
  simTarget=GPUComputeSim
  emitterUpdateScript [
   &EmitterState(
    Enable Distance Culling=true
    Life Cycle Mode=::NewEnumerator1
    Loop Behavior=::NewEnumerator1
    Loop Duration=9.5
    Module.bUseMaxDistance=true
    Module.bUseMinDistance=false
    Scalability Mode=::NewEnumerator1
    UseLoopDelay=false
   )
   &SpawnRate(
    SpawnRate=&FloatFromCurve(FloatCurve=DataInterface() {propertyValues {Curve {keys [<at 0 0 RCIM_Cubic>,<at 4 350>]}}}, CurveIndex=$Emitter.Age)
    Use Spawn Probability=false
   )
  ]
  particleSpawnScript [
   &InitializeParticle(
    Lifetime=&RandomRangeFloat(Minimum=0.3, Maximum=0.5, Evaluation Type=::NewEnumerator1)
    Module.Write Color=true
    Module.Write Lifetime=true
    Module.Write Mass=true
    Module.Write Material Random=true
    Module.Write Position=false
    Module.Write Scale=false
    Module.Write SpriteRotation=false
    Module.Write SpriteSize=true
    Sprite Size=&RandomRangeVector2D(Minimum=<1.2 1.2>, Maximum=<2.4 2.4>, Evaluation Type=::NewEnumerator1)
   )
   &Color(Color=&MakeLinearColorFromVectorAndFloat("Vector (RGB)"=<4.165 24.7513 25>, "Float (Alpha)"=&ClampFloat(Float=2)))
   &SphereLocation(
    MaskSpawnGroup=false
    Module.Override Randomness=false
    Sphere Radius=&RandomRangeFloat(Minimum=150, Maximum=210, Evaluation Type=::NewEnumerator1)
    Surface Only=true
   )
   &AddVelocity(Velocity=&Multiply_VectorByFloat(
     Vector=$Output.SphereLocation.SphereVector
     Float=&RandomRangeFloat(Minimum=0, Maximum=0.1, Evaluation Type=::NewEnumerator1)
    ))
  ]
  particleUpdateScript [
   &ParticleState(DeltaTime:NiagaraFloat=&Divide_Float(A=$Engine.DeltaTime, B=$User.Time Scale))
   &Drag(Use Linear Drag=true, Use Rotational Drag=false)
   &PointAttractionForce(
    Attraction Radius=500
    Attraction Strength=&FloatFromCurve(
     FloatCurve=DataInterface() {propertyValues {Curve {keys [{interpMode=RCIM_Cubic, tangentMode=RCTM_Break, time=0, value=0}, <at 9 500>]}}}
     CurveIndex=$Emitter.Age
    )
    Falloff Exponent=0.4
    Kill Radius=0.5
    Kill Within Radius=true
    Module.Use Falloff=true
    UseAttractorOffset=false
   )
   &SolveForcesAndVelocity(Clamp Velocity=false, Limit Acceleration=false)
   &ScaleSpriteSize(Scale Factor=&Vector2DFromFloat(Value=&FloatFromCurve(
      FloatCurve=DataInterface() {propertyValues {Curve {keys [
          <at -0.0044 0 RCIM_Cubic>
          {interpMode=RCIM_Cubic, tangentMode=RCTM_User, time=0.157, value=1}
          {interpMode=RCIM_Cubic, tangentMode=RCTM_User, time=0.7034, value=1}
          <at 1 0>
         ]}}}
      Scale Curve=$System.Scale_Closing
     )))
   &ScaleSpriteSizeBySpeed(
    Max Scale Factor=<1 20>
    Min Scale Factor=<1 3>
    Source Velocity:Vector3f=&Multiply_Vector(A=<0 0.01 1>, B=$Particles.Velocity)
    Velocity Threshold=750
   )
   &ScaleSpriteSize(Scale Factor=&MakeVector2D(
     X=&FloatFromCurve(FloatCurve=DataInterface() {propertyValues {Curve {keys [<at 0 1>,<at 15000 12>]}}}, CurveIndex=$Engine.Owner.LODDistance, Scale Curve=1.5)
     Y=1.5
    )) {module=ScaleSpriteSize001}
  ]
  rendererClasses [&NiagaraSpriteRendererProperties]
  renderers [{rendererClass=&NiagaraSpriteRendererProperties}]
 }
 FloatingFrags {
  emitterUpdateScript [
   &EmitterState(
    Enable Distance Culling=true
    MaxDistance=2000
    MinDistance=1000
    Module.bUseMaxDistance=true
    Module.bUseMinDistance=false
    Scalability Mode=::NewEnumerator1
   )
   &SpawnBurst_Instantaneous(Spawn Count=1, Spawn Time=0.25, Use Spawn Probability=false)
   &SpawnBurst_Instantaneous(Spawn Count=1, Spawn Time=0.7, Use Spawn Probability=false) {module=SpawnBurst_Instantaneous001}
   &SpawnBurst_Instantaneous(Spawn Count=1, Spawn Time=1.4, Use Spawn Probability=false) {module=SpawnBurst_Instantaneous002}
  ]
  particleSpawnScript [
   /SetVariables(Particles.MeshOrientation=&MakeQuatFromAxisAngle(
     Axis=&RandomRangeVector(Minimum=<0 0 0>, Maximum=<1 1 1>, Evaluation Type=::NewEnumerator1)
     AngleInDegrees=&RandomRangeFloat(Minimum=0, Maximum=180)
    )) {isSetParameters=true}
   /SetVariables(Particles.SpawnRotationVector=&RandomRangeVector(Minimum=<-0.01 -0.01 -0.01>, Maximum=<0.01 0.01 0.01>, Evaluation Type=::NewEnumerator1)) {isSetParameters=true}
   &InitializeParticle(
    Color=<rgba 0.7277 0.7277 1 0.5>
    Lifetime=&FloatFromCurve(FloatCurve=DataInterface() {propertyValues {Curve {keys [<at 0 10>,<at 2 6>]}}}, CurveIndex=$System.Age)
    Mesh Scale=&VectorFromFloat(Value=&RandomRangeFloat(Minimum=4.8, Maximum=6, Evaluation Type=::NewEnumerator1))
    Module.Write Color=true
    Module.Write Lifetime=true
    Module.Write Mass=true
    Module.Write Material Random=true
    Module.Write Position=false
    Module.Write Scale=true
    Module.Write SpriteRotation=false
    Module.Write SpriteSize=false
   )
  ]
  particleUpdateScript [
   &ParticleState()
   &DynamicMaterialParameters(
    Index 0 Param 1=1
    Index 0 Param 2=1
    Index 0 Param 3=1
    Index 0 Param 4=&FloatFromCurve(FloatCurve=DataInterface() {propertyValues {Curve {keys [
         <at 0 -1.5 RCIM_Cubic>
         {arriveTangent=2.8362, interpMode=RCIM_Cubic, leaveTangent=2.8362, time=0.0658, value=-1.18}
         {arriveTangent=2.8159, interpMode=RCIM_Cubic, leaveTangent=2.8159, time=0.2547, value=-0.7776}
         {arriveTangent=2.7086, interpMode=RCIM_Cubic, leaveTangent=2.7086, time=0.2776, value=-0.5837}
         {arriveTangent=2.368, interpMode=RCIM_Cubic, leaveTangent=2.368, time=0.4377, value=-0.2819}
         {arriveTangent=1.7079, interpMode=RCIM_Cubic, leaveTangent=1.7079, time=0.4731, value=-0.1206}
         {arriveTangent=1.9231, interpMode=RCIM_Cubic, leaveTangent=1.9231, time=0.8103, value=0.3545}
         {arriveTangent=3.4034, interpMode=RCIM_Cubic, leaveTangent=3.4034, time=0.8496, value=0.6032}
         <at 1 1>
        ]}}})
    Module.Param0WriteEnabled=true
    Module.Param1WriteEnabled=true
    Module.Param2WriteEnabled=true
    Module.Param3WriteEnabled=true
   )
   &ScaleMeshSize(Scale Factor=&VectorFromFloat(Value=&FloatFromCurve(
      FloatCurve=DataInterface() {propertyValues {Curve {keys [
          {arriveTangent=-0.0466, time=0, value=0.6}
          {arriveTangent=0.0413, interpMode=RCIM_Cubic, leaveTangent=0.0413, time=0.2603, value=0.5534}
          {arriveTangent=0.7472, interpMode=RCIM_Cubic, leaveTangent=0.7472, time=0.2832, value=0.6117}
          {arriveTangent=0.673, interpMode=RCIM_Cubic, leaveTangent=0.673, time=0.4449, value=0.6913}
          {arriveTangent=0.4479, interpMode=RCIM_Cubic, leaveTangent=0.4479, time=0.4705, value=0.7378}
          {arriveTangent=0.4597, interpMode=RCIM_Cubic, leaveTangent=0.4597, time=0.8146, value=0.8569}
          {arriveTangent=-4.6223, interpMode=RCIM_Cubic, leaveTangent=-4.6223, time=0.8441, value=0.9095}
          <at 0.98 1 RCIM_Cubic>
          <at 1 0>
         ]}}}
      Scale Curve=$System.Scale_Closing
     )))
   &ScaleColor(
    Scale Alpha=&FloatFromCurve(FloatCurve=DataInterface() {propertyValues {Curve {keys [<at 0 0 RCIM_Cubic>, {arriveTangent=2, interpMode=RCIM_Cubic, leaveTangent=2, time=0.1598, value=0.3945}, <at 0.5 1>]}}})
    Scale RGB=&VectorFromCurve(VectorCurve=DataInterface() {propertyValues {
       XCurve {keys [<at 0.5 2 RCIM_Cubic>,<at 1 4.24>]}
       YCurve {keys [<at 0.5 2 RCIM_Cubic>,<at 1 5.8206>]}
       ZCurve {keys [<at 0.5 2 RCIM_Cubic>,<at 1 6>]}
      }})
    ScaleA=true
    ScaleRGB=true
   )
   &UpdateMeshOrientation(Rotation Vector=$Particles.SpawnRotationVector) {module=UpdateMeshOrientation001}
   &CameraOffset(Camera Offset Amount=30)
  ]
  rendererClasses [&NiagaraMeshRendererProperties]
  renderers [{rendererClass=&NiagaraMeshRendererProperties, rendererIndex=0}]
 }
 Light001 {rendererClasses [], renderers []}
 Rift_ {
  emitterUpdateScript [
   &EmitterState(Inactive Response=::NewEnumerator1, Life Cycle Mode=::NewEnumerator1, Loop Behavior=::NewEnumerator1, Loop Duration=10, UseLoopDelay=false)
   /SetVariables(Emitter.Scale:Vector3f=&Lerp_Vector(
     A=<0.1 0.1 0.1>
     B=<1 1 1>
     Alpha=&FloatFromCurve(
      FloatCurve=DataInterface() {propertyValues {Curve {keys [
          {interpMode=RCIM_Cubic, tangentMode=RCTM_User, time=0, value=0}
          {arriveTangent=0.125, interpMode=RCIM_Cubic, leaveTangent=0.125, time=5, value=0.33}
          {interpMode=RCIM_Cubic, tangentMode=RCTM_User, time=8, value=1}
          <at 9.8 1>
          {interpMode=RCIM_Cubic, tangentMode=RCTM_User, time=9.9, value=1.1}
          {interpMode=RCIM_Cubic, leaveTangent=-4.944, time=10, value=0}
         ]}}}
      CurveIndex=$Emitter.Age
     )
    )) {isSetParameters=true}
   &SpawnBurst_Instantaneous(Use Spawn Probability=false, UseLoopCountLimit=false)
  ]
  particleSpawnScript [/Niagara/Modules/Spawn/Initialization/V2/InitializeParticle(Lifetime=10, Mesh Scale Mode=::NewEnumerator3, UsePositionOffset=false, Write Material Random=true)]
  particleUpdateScript [&ParticleState()]
  rendererClasses [&NiagaraMeshRendererProperties]
  renderers [{rendererClass=&NiagaraMeshRendererProperties, rendererIndex=0}]
 }
 RoundFlare {rendererClasses [], renderers []}
 ShockWave {
  emitterUpdateScript [
   &EmitterState(Life Cycle Mode=::NewEnumerator1, Loop Behavior=::NewEnumerator1, Loop Delay=$System.Burst Delay, Loop Duration=1, UseLoopDelay=true)
   &SpawnBurst_Instantaneous(Spawn Count=1, Use Spawn Probability=false)
  ]
  particleSpawnScript [
   /Niagara/Modules/Spawn/Initialization/InitializeParticle(
    Lifetime=0.3
    Mesh Scale=<0.45 0.45 0.45>
    Module.Write Color=true
    Module.Write Lifetime=true
    Module.Write Mass=true
    Module.Write Material Random=true
    Module.Write Position=false
    Module.Write Scale=true
    Module.Write SpriteRotation=false
    Module.Write SpriteSize=false
   )
   &Color(Color=&MakeLinearColorFromVectorAndFloat("Vector (RGB)"=<1 1 1>, "Float (Alpha)"=&ClampFloat(Float=0.1)))
  ]
  particleUpdateScript [
   &ParticleState()
   &ScaleColor(
    Scale Alpha=&FloatFromCurve(FloatCurve=DataInterface() {propertyValues {Curve {keys [{arriveTangent=-2.1694, interpMode=RCIM_Cubic, time=0, value=0}, <at 0.4 1 RCIM_Cubic>, <at 1 0 RCIM_Cubic>]}}})
    ScaleA=true
    ScaleRGB=true
   )
   &ScaleMeshSize(Scale Factor=&VectorFromFloat(Value=&FloatFromCurve(FloatCurve=DataInterface() {propertyValues {Curve {keys [<at 0 1 RCIM_Cubic>,<at 1 0>]}}})))
   &CameraOffset(Camera Offset Amount=120)
  ]
  rendererClasses [&NiagaraMeshRendererProperties]
  renderers [{rendererClass=&NiagaraMeshRendererProperties, rendererIndex=0}]
 }
 SmokeInhale {rendererClasses [], renderers []}
 WideFlare001 {rendererClasses [], renderers []}
}
summary {
 systemName=NS_Redux_Portal_Closing
 emitters [
  {emitterName=CrackSpline_, rendererClasses [&NiagaraMeshRendererProperties]}
  {emitterName=FloatingFrags, rendererClasses [&NiagaraMeshRendererProperties]}
  {emitterName=ElectricArcsResidual_, rendererClasses [&NiagaraMeshRendererProperties]}
  {bEnabled=false, emitterName=TinyFrags, rendererClasses [&NiagaraMeshRendererProperties]}
  {emitterName=Embers_, simTarget=GPUComputeSim, rendererClasses [&NiagaraSpriteRendererProperties]}
  {emitterName=Corona, rendererClasses [&NiagaraSpriteRendererProperties, &NiagaraLightRendererProperties]}
  {bEnabled=false, emitterName=RoundFlare_, rendererClasses [&NiagaraSpriteRendererProperties]}
  {bEnabled=false, emitterName=WideFlare001_, rendererClasses [&NiagaraSpriteRendererProperties]}
  {bEnabled=false, emitterName=Light001_, rendererClasses [&NiagaraSpriteRendererProperties]}
  {emitterName=ShockWave, rendererClasses [&NiagaraMeshRendererProperties]}
  {bEnabled=false, emitterName=SmokeInhale_, rendererClasses [&NiagaraSpriteRendererProperties]}
  {emitterName=Rift_, rendererClasses [&NiagaraMeshRendererProperties]}
  {emitterName=RoundFlare, rendererClasses []}
  {emitterName=WideFlare001, rendererClasses []}
  {emitterName=Light001, rendererClasses []}
  {emitterName=SmokeInhale, rendererClasses []}
 ]
 userVariables [User.Time Scale {type=NiagaraFloat, defaultValue=1, description=""}]
}
