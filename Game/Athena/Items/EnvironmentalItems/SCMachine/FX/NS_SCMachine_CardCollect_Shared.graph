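# Niagara effect shared for SC Machine card collection, showing a short-lived burst drawn as both a cyan mesh and a sprite.
# Curves set position directly, lifting it about 100 units with a mid-life dip before dropping back.
# The mesh opens and collapses along Y, the sprite swells and shrinks, alpha fades, and a curve animates dynamic material parameters.
# A User.Purchased bool is exposed, alongside a bool-to-color dynamic input dependency.
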NiagaraSystem /Game/Athena/Items/EnvironmentalItems/SCMachine/FX/NS_SCMachine_CardCollect_Shared
dependencies {
 usedDataInterfaces [&NiagaraDataInterfaceCurve, &NiagaraDataInterfaceVectorCurve]
 usedDynamicInputs [&MakeCustomLinearColorFromBool]
 usedModules [
  &SystemState
  &self:SystemSpawnScript.SystemScriptSource.SystemScriptGraph.NiagaraNodeAssignment_0.SetVariables_#1
 ]
}
emitters {Empty {
  emitterUpdateScript [
   &EmitterState()
   &SpawnBurst_Instantaneous(Use Spawn Probability=false, UseLoopCountLimit=false)
  ]
  particleSpawnScript [
   &InitializeParticle(
    Color=<rgba 0.1221 1.9555 3 1>
    Lifetime=1.16
    Mesh Scale Mode=::NewEnumerator3
    Mesh Uniform Scale=2
    Sprite Size Mode=::NewEnumerator3
    Uniform Sprite Size=50
    UsePositionOffset=false
    Write Material Random=true
   )
   &InitialMeshOrientation(Mesh Orientation Mode=::NewEnumerator2, Orientation Vector=<0 0 1>, Use Rotation Vector=false)
  ]
  particleUpdateScript [
   &ParticleState()
   &ScaleSpriteSize(Uniform Curve Sprite Scale=DataInterface() {propertyValues {Curve {keys [
        {interpMode=RCIM_Cubic, tangentMode=RCTM_User, time=0, value=0}
        {interpMode=RCIM_Cubic, tangentMode=RCTM_User, time=0.5, value=1}
        {tangentMode=RCTM_User, time=1, value=0}
       ]}}})
   &ScaleMeshSize(Scale Factor=&VectorFromCurve(VectorCurve=DataInterface() {propertyValues {
       XCurve {keys [{interpMode=RCIM_Cubic, tangentMode=RCTM_User, time=0, value=1}]}
       YCurve {keys [
         {interpMode=RCIM_Cubic, tangentMode=RCTM_User, time=0, value=0}
         {arriveTangent=0.2524, interpMode=RCIM_Cubic, leaveTangent=0.2524, tangentMode=RCTM_User, time=0.2, value=1}
         {interpMode=RCIM_Cubic, tangentMode=RCTM_User, time=0.75, value=1}
         <at 1 0 RCIM_Cubic>
        ]}
       ZCurve {keys [<at 0 1>,<at 1 1>]}
      }}))
   &DynamicMaterialParameters(
    Index 0 Param 2=&FloatFromCurve(
     FloatCurve=DataInterface() {propertyValues {Curve {keys [
         {tangentMode=RCTM_User, time=-1, value=0}
         {interpMode=RCIM_Cubic, tangentMode=RCTM_User, time=0, value=-1}
         {interpMode=RCIM_Cubic, tangentMode=RCTM_User, time=0.5, value=1}
        ]}}}
     Scale Curve=25
    )
    Index 1 Param 1=15
    Module.Index0Param0Write=true
    Module.Index0Param1Write=true
    Module.Index0Param2Write=true
    Module.Index0Param3Write=true
    Module.Index1Param0Write=true
    Module.Index1Param1Write=true
    Module.Index1Param2Write=true
    Module.Index1Param3Write=true
    Write Parameter Index 1=true
   )
   /SetVariables(
    Particles.DynamParam_Sprite:Vector4f=<1 0 0 0>
    Particles.Position=&VectorFromCurve(
     VectorCurve=DataInterface() {propertyValues {
       XCurve {keys [<at 0 1>,<at 1 0>]}
       YCurve {keys [<at 0 1>,<at 1 0>]}
       ZCurve {keys [
         {interpMode=RCIM_Cubic, tangentMode=RCTM_User, time=0, value=0}
         {arriveTangent=-0.0143, interpMode=RCIM_Cubic, leaveTangent=-0.0143, tangentMode=RCTM_User, time=0.2, value=0.95}
         <at 0.5 0.85 RCIM_Cubic>
         {arriveTangent=-0.2108, interpMode=RCIM_Cubic, leaveTangent=-0.2108, tangentMode=RCTM_User, time=0.8, value=0.95}
         {tangentMode=RCTM_User, time=1, value=0}
        ]}
      }}
     Scale Curve=<1 1 100>
    )
   ) {isSetParameters=true}
   &ScaleColor(
    Scale Alpha=&FloatFromCurve(FloatCurve=DataInterface() {propertyValues {Curve {keys [<at 0 1>,<at 1 0>]}}})
    ScaleA=true
    ScaleRGB=true
   )
  ]
  rendererClasses [&NiagaraMeshRendererProperties, &NiagaraSpriteRendererProperties]
  renderers [
   {rendererClass=&NiagaraMeshRendererProperties, rendererIndex=0}
   {rendererClass=&NiagaraSpriteRendererProperties, rendererIndex=1}
  ]
 }}
summary {
 systemName=NS_SCMachine_CardCollect_Shared
 emitters [{emitterName=Empty, rendererClasses [&NiagaraMeshRendererProperties, &NiagaraSpriteRendererProperties]}]
 userVariables [User.Purchased {type=NiagaraBool, defaultValue=true}]
}
