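# Niagara system for a holographic battery effect on the SCMachine environmental item.
# Six emitters, all suffixed LW, cover glow, holo beams, pedestal light, squares, and a holo logo split into start and loop emitters.
# User floats GameTime and InteractDuration are exposed, apparently for interaction timing.
# SystemState is the only listed module; no renderers or emitter internals appear.
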NiagaraSystem /Game/Athena/Items/EnvironmentalItems/SCMachine/FX/NS_SCMachine_HoloBattery_LW
dependencies {usedDataInterfaces [], usedDynamicInputs [], usedModules [&SystemState]}
emitters {
 Glow_LW {rendererClasses [], renderers []}
 HoloBeams_LW {rendererClasses [], renderers []}
 HoloLogoLoop_LW {rendererClasses [], renderers []}
 HoloLogoStart_LW {rendererClasses [], renderers []}
 PedestalLight_LW {rendererClasses [], renderers []}
 Squares_LW {rendererClasses [], renderers []}
}
summary {
 systemName=NS_SCMachine_HoloBattery_LW
 emitters [
  {emitterName=Glow_LW, rendererClasses []}
  {emitterName=HoloBeams_LW, rendererClasses []}
  {emitterName=PedestalLight_LW, rendererClasses []}
  {emitterName=Squares_LW, rendererClasses []}
  {emitterName=HoloLogoStart_LW, rendererClasses []}
  {emitterName=HoloLogoLoop_LW, rendererClasses []}
 ]
 userVariables [
  User.GameTime {type=NiagaraFloat, defaultValue=0, description=""}
  User.InteractDuration {type=NiagaraFloat, defaultValue=4, description=""}
 ]
}
