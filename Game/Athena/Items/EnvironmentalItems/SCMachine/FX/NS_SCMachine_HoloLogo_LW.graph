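# Niagara system for the SCMachine environmental item's hologram logo effect, with each emitter paired with an _LW counterpart (likely lightweight).
# The four original emitters are disabled, and the active _LW versions show no contents here.
# In the originals, a mesh logo rises, grows and fades in, and a looping logo animates material parameters.
# Blue beam sprites drift upward, while a brief GPU burst of squares shoots up from a cylinder.
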NiagaraSystem /Game/Athena/Items/EnvironmentalItems/SCMachine/FX/NS_SCMachine_HoloLogo_LW
dependencies {
 usedDataInterfaces [
  &NiagaraDataInterfaceCurve
  &NiagaraDataInterfaceVectorCurve
  &NiagaraDataInterfaceVector2DCurve
 ]
 usedDynamicInputs []
 usedModules [&SystemState]
}
emitters {
 HoloBeams {
  bEnabled=false
  emitterUpdateScript [
   &EmitterState(Inactive Response=::NewEnumerator1, Life Cycle Mode=::NewEnumerator1, Loop Duration=0.3, UseLoopDelay=true)
   &SpawnBurst_Instantaneous(Use Spawn Probability=false, UseLoopCountLimit=false)
  ]
  particleSpawnScript [
   &InitializeParticle(
    Lifetime=&RandomRangeFloat(Minimum=0.5, Maximum=0.8)
    Position=&ConvertVectorToPosition(
     Input Position=&RandomRangeVector(Minimum=<0 0 -20>, Maximum=<0 0 -15>)
     Conversion Operation:ENiagara_LWCConvertVecToPos=::NewEnumerator0
    )
    Position Mode=::NewEnumerator1
    Sprite Size=&RandomRangeVector2D(Minimum=<220 140>, Maximum=<260 200>)
    Sprite Size Mode=::NewEnumerator4
    UsePositionOffset=false
    Write Material Random=true
   )
   &Color(Color=&MakeLinearColorFromVectorAndFloat("Vector (RGB)"=<0.4083 1.0179 2>, "Float (Alpha)"=&ClampFloat(Float=0.4)))
   &AddVelocity(Velocity=<0 0 20>)
  ]
  particleUpdateScript [
   &ParticleState()
   /Niagara/Modules/Update/SubUV/SubUVAnimation(Number Of Frames=0)
   &ScaleColor(
    Scale Alpha=&FloatFromCurve(FloatCurve=DataInterface() {propertyValues {Curve {keys [
         <at 0 0 RCIM_Cubic>
         {arriveTangent=1.6667, interpMode=RCIM_Cubic, leaveTangent=1.6667, time=0.4, value=1}
         {arriveTangent=-1.6667, interpMode=RCIM_Cubic, leaveTangent=-1.6667, time=0.6, value=1}
         <at 1 0>
        ]}}})
    ScaleA=true
    ScaleRGB=true
   )
   &SolveForcesAndVelocity(Clamp Velocity=false, Limit Acceleration=false)
   &CameraOffset(Camera Offset Amount=35)
  ]
  rendererClasses [&NiagaraSpriteRendererProperties]
  renderers [{rendererClass=&NiagaraSpriteRendererProperties}]
 }
 HoloBeams_LW {rendererClasses [], renderers []}
 HoloLogoLoop {
  bEnabled=false
  emitterUpdateScript [
   &EmitterState(Inactive Response=::NewEnumerator1, Life Cycle Mode=::NewEnumerator1, Loop Delay=0.2, UseLoopDelay=true)
   &SpawnBurst_Instantaneous(Use Spawn Probability=false, UseLoopCountLimit=false)
  ]
  particleSpawnScript [
   &InitializeParticle(
    Lifetime=5
    Mesh Scale Mode=::NewEnumerator4
    Position=&ConvertVectorToPosition(Input Position=<0 0 200>, Conversion Operation:ENiagara_LWCConvertVecToPos=::NewEnumerator0)
    Position Mode=::NewEnumerator1
    UsePositionOffset=false
    Write Material Random=true
   )
   &Color(Color=&MakeLinearColorFromVectorAndFloat("Vector (RGB)"=<1 1 1>, "Float (Alpha)"=&ClampFloat(Float=1)))
  ]
  particleUpdateScript [
   &ParticleState()
   &ScaleColor(ScaleA=true, ScaleRGB=true)
   &DynamicMaterialParameters(
    Index 0 Param 1=&FloatFromCurve(FloatCurve=DataInterface() {propertyValues {Curve {keys [<at 0 0 RCIM_Cubic>, {arriveTangent=3.3333, interpMode=RCIM_Cubic, leaveTangent=3.3333, time=0.0553, value=0.0563}, <at 0.3 1>]}}})
    Index 0 Param 2=1
    Index 0 Param 4=4
    Module.Index0Param0Write=true
    Module.Index0Param1Write=true
    Module.Index0Param2Write=true
    Module.Index0Param3Write=true
   )
  ]
  rendererClasses [&NiagaraMeshRendererProperties]
  renderers [{rendererClass=&NiagaraMeshRendererProperties, rendererIndex=0}]
 }
 HoloLogoLoop_LW {rendererClasses [], renderers []}
 HoloLogoStart {
  bEnabled=false
  emitterUpdateScript [
   &EmitterState(Inactive Response=::NewEnumerator1, Life Cycle Mode=::NewEnumerator1, Loop Behavior=::NewEnumerator1, Loop Duration=1, UseLoopDelay=true)
   &SpawnBurst_Instantaneous(Use Spawn Probability=false, UseLoopCountLimit=false)
  ]
  particleSpawnScript [
   &InitializeParticle(Lifetime=0.2, Mesh Scale Mode=::NewEnumerator4, UsePositionOffset=false, Write Material Random=true)
   &Color(Color=&MakeLinearColorFromVectorAndFloat("Vector (RGB)"=<1 1 1>, "Float (Alpha)"=&ClampFloat(Float=1)))
  ]
  particleUpdateScript [
   /SetVariables(Particles.Position=&VectorFromCurve(VectorCurve=DataInterface() {propertyValues {
       XCurve {keys [<at 0 0 RCIM_Cubic>,<at 1 0>]}
       YCurve {keys [<at 0 0 RCIM_Cubic>,<at 1 0>]}
       ZCurve {keys [{arriveTangent=239.7967, interpMode=RCIM_Cubic, time=0, value=0}, <at 1 100>]}
      }})) {isSetParameters=true}
   &ParticleState()
   &ScaleMeshSize(Scale Factor=&VectorFromFloat(Value=&FloatFromCurve(FloatCurve=DataInterface() {propertyValues {Curve {keys [{arriveTangent=1.5597, interpMode=RCIM_Cubic, time=0, value=0.4}, <at 1 1>]}}})))
   &ScaleColor(
    Scale Alpha=&FloatFromCurve(FloatCurve=DataInterface() {propertyValues {Curve {keys [<at 0 0 RCIM_Cubic>,<at 1 1>]}}})
    ScaleA=true
    ScaleRGB=true
   )
  ]
  rendererClasses [&NiagaraMeshRendererProperties]
  renderers [{rendererClass=&NiagaraMeshRendererProperties, rendererIndex=0}]
 }
 HoloLogoStart_LW {rendererClasses [], renderers []}
 Squares {
  bEnabled=false
  simTarget=GPUComputeSim
  emitterUpdateScript [
   &EmitterState(Life Cycle Mode=::NewEnumerator1, Loop Behavior=::NewEnumerator1, Loop Duration=0.07, UseLoopDelay=true)
   &SpawnBurst_Instantaneous(Spawn Count=15, Use Spawn Probability=false, UseLoopCountLimit=false)
   &SpawnRate(SpawnRate=150, Use Spawn Probability=false)
  ]
  particleSpawnScript [
   &InitializeParticle(
    Lifetime=&RandomRangeFloat(Minimum=0.15, Maximum=0.3)
    Sprite Size=&RandomRangeVector2D(Minimum=<1.8 7>, Maximum=<2.5 8>)
    Sprite Size Mode=::NewEnumerator4
    UsePositionOffset=false
    Write Material Random=true
   )
   &Color(Color=&MakeLinearColorFromVectorAndFloat("Vector (RGB)"=<0.0928 0.6741 1>, "Float (Alpha)"=&ClampFloat(Float=1)))
   &ShapeLocation(
    Cylinder Radius=&RandomRangeFloat(Minimum=40, Maximum=55)
    MaskSpawnGroup=false
    Offset=<0 0 25>
    Offset Mode=::NewEnumerator0
    Override Seed=false
    Rotation Axis=<0 0 1>
    Rotation Mode=::NewEnumerator1
    Shape Primitive=::NewEnumerator1
    Surface Only=true
    UseOwnerScale=true
   )
   &AddVelocity(Velocity=&RandomRangeVector(Minimum=<0 0 400>, Maximum=<0 0 800>))
  ]
  particleUpdateScript [
   &ParticleState()
   &ScaleSpriteSize(
    Scale Factor=&Vector2DFromCurve(
     Vector2Curve=DataInterface() {propertyValues {
       XCurve {keys [{arriveTangent=-4.1323, interpMode=RCIM_Cubic, time=0, value=2}, <at 0.8 1>]}
       YCurve {keys [{arriveTangent=-43.3352, interpMode=RCIM_Cubic, time=0, value=15}, <at 0.8 1>]}
      }}
     ScaleCurve=<1 1>
    )
    Scale Sprite Size Mode=::NewEnumerator2
   )
   &ScaleColor(
    Scale Alpha=&FloatFromCurve(FloatCurve=DataInterface() {propertyValues {Curve {keys [{arriveTangent=5.9709, interpMode=RCIM_Cubic, time=0, value=0}, <at 0.2 0.5 RCIM_Cubic>, <at 1 0>]}}})
    ScaleA=true
    ScaleRGB=true
   )
   &Drag(Drag=4, Use Linear Drag=true, Use Rotational Drag=true)
   &SolveForcesAndVelocity(Clamp Velocity=false, Limit Acceleration=false)
  ]
  rendererClasses [&NiagaraSpriteRendererProperties]
  renderers [{rendererClass=&NiagaraSpriteRendererProperties}]
 }
 Squares_LW {rendererClasses [], renderers []}
}
summary {
 systemName=NS_SCMachine_HoloLogo_LW
 emitters [
  {bEnabled=false, emitterName=HoloBeams, rendererClasses [&NiagaraSpriteRendererProperties]}
  {bEnabled=false, emitterName=HoloLogoStart, rendererClasses [&NiagaraMeshRendererProperties]}
  {bEnabled=false, emitterName=HoloLogoLoop, rendererClasses [&NiagaraMeshRendererProperties]}
  {bEnabled=false, emitterName=Squares, simTarget=GPUComputeSim, rendererClasses [&NiagaraSpriteRendererProperties]}
  {emitterName=HoloLogoLoop_LW, rendererClasses []}
  {emitterName=Squares_LW, rendererClasses []}
  {emitterName=HoloLogoStart_LW, rendererClasses []}
  {emitterName=HoloBeams_LW, rendererClasses []}
 ]
 userVariables [
  User.GameTime {type=NiagaraFloat, defaultValue=0, description=""}
  User.InteractDuration {type=NiagaraFloat, defaultValue=4, description=""}
 ]
}
