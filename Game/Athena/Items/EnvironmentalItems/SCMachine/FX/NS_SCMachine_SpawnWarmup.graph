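# Niagara warmup effect for the SCMachine environmental item's spawn, a vertical beam effect tinted blue or orange through the User.Variant integer.
# Mesh beams include one that grows and fades in with system age and a short-lived darker one pushed by camera offset.
# Ring meshes descend along a position curve, flattening and fading near the bottom.
# GPU square sprites fall from a cylinder surface at an age-ramped spawn rate, stretched by speed.
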NiagaraSystem /Game/Athena/Items/EnvironmentalItems/SCMachine/FX/NS_SCMachine_SpawnWarmup
dependencies {
 usedDataInterfaces [
  &NiagaraDataInterfaceCurve
  &NiagaraDataInterfaceVector2DCurve
  &NiagaraDataInterfaceVectorCurve
  &NiagaraDataInterfaceArrayColor
 ]
 usedDynamicInputs [&SelectLinearColorFromArray, &SetBoolByIntComparison]
 usedModules [
  &self:SystemSpawnScript.SystemScriptSource.SystemScriptGraph.NiagaraNodeAssignment_0.SetVariables_#1
  &SystemState
  &Timeline
 ]
}
emitters {
 Beam {
  emitterUpdateScript [
   &EmitterState(Life Cycle Mode=::NewEnumerator1, Loop Behavior=::NewEnumerator1, Loop Duration Mode=::NewEnumerator1, UseLoopDelay=false)
   &SpawnBurst_Instantaneous(Spawn Time=0.5, Use Spawn Probability=false, UseLoopCountLimit=false)
  ]
  particleSpawnScript [&InitializeParticle(Mesh Scale Mode=::NewEnumerator4, UsePositionOffset=false, Write Material Random=true)]
  particleUpdateScript [
   &ParticleState(
    Kill Particles When Lifetime Has Elapsed=false
    Let Infinitely Lived Particles Die When Emitter Deactivates=true
    Loop Particles Lifetime=true
   )
   &ScaleMeshSize(Scale Factor=&VectorFromCurve(
     VectorCurve=DataInterface() {propertyValues {
       XCurve {keys [{tangentMode=RCTM_User, time=0, value=0}, {tangentMode=RCTM_User, time=0.75, value=1}]}
       YCurve {keys [{tangentMode=RCTM_User, time=0, value=0}, {tangentMode=RCTM_User, time=0.75, value=1}]}
       ZCurve {keys [{tangentMode=RCTM_User, time=0, value=0}, {tangentMode=RCTM_User, time=0.75, value=1}]}
      }}
     CurveIndex=$System.Age
    ))
   &DynamicMaterialParameters(
    Index 0 Param 1=5000
    Index 0 Param 2=1
    Index 0 Param 3=1
    Index 0 Param 4=1
    Module.Index0Float0Write=true
    Module.Index0Float1Write=true
    Module.Index0Float2Write=true
    Module.Index0Float3Write=true
   )
   &ScaleColor(
    Scale Alpha=&FloatFromCurve(
     FloatCurve=DataInterface() {propertyValues {Curve {keys [{tangentMode=RCTM_User, time=0, value=0}, {tangentMode=RCTM_User, time=1, value=1}]}}}
     CurveIndex=$System.Age
    )
    Scale RGB=&Lerp_Vector(
     A=&MakeVectorFromLinearColorRGB(Color=<rgba 0 0.4762 1 1>)
     B=&MakeVectorFromLinearColorRGB(Color=<rgba 0.7016 0.3638 0 0.5>)
     Alpha=&MakeFloatFromInt(Int=$User.Variant)
    )
    ScaleA=true
    ScaleRGB=true
   )
  ]
  rendererClasses [&NiagaraMeshRendererProperties]
  renderers [
   {rendererClass=&NiagaraMeshRendererProperties, rendererIndex=0}
   {rendererClass=&NiagaraMeshRendererProperties, rendererIndex=1}
  ]
 }
 PickupBeamDark {
  emitterUpdateScript [
   &EmitterState(Inactive Response=::NewEnumerator1, Life Cycle Mode=::NewEnumerator1, Loop Duration=1.5, UseLoopDelay=true)
   &SpawnBurst_Instantaneous(Spawn Time=0.3, Use Spawn Probability=false, UseLoopCountLimit=false)
  ]
  particleSpawnScript [
   &InitializeParticle(
    Lifetime=2
    Mesh Scale=<5 1 1>
    Mesh Scale Mode=::NewEnumerator4
    Position Offset=<0 0 -75>
    UsePositionOffset=true
    Write Material Random=true
   )
   &InitialMeshOrientation(Mesh Orientation Mode=::NewEnumerator2, Orientation Vector=<0 0 1>, Use Rotation Vector=false)
  ]
  particleUpdateScript [
   &ParticleState()
   /SetVariables(
    Particles.BlueBeam_Color:LinearColor=/Niagara/DynamicInputs/Lerp/SmoothLerpOverTime_Color(
     Smooth Value:LinearColor=<rgba 0.1624 0.7479 1 0.5>
     Target Value:LinearColor=<rgba 1 0.537 0 0.5>
     Convergence Rate:NiagaraFloat=&MakeFloatFromInt(Int=$User.Variant)
    )
    Particles.BlueBeam_DynParam:Vector4f=<110 1 1 1>
   ) {isSetParameters=true}
   &ScaleColor(
    Scale Alpha=&FloatFromCurve(FloatCurve=DataInterface() {propertyValues {Curve {keys [<at 0 0 RCIM_Cubic>, <at 0.3 1>, {arriveTangent=-1.4286, interpMode=RCIM_Cubic, leaveTangent=-1.4286, time=0.7, value=1}, <at 1 0>]}}})
    Scale RGB=&Lerp_Vector(
     A=&MakeVectorFromLinearColorRGB(Color=<rgba 0 0.1807 0.2487 0.3>)
     B=&MakeVectorFromLinearColorRGB(Color=<rgba 0.5789 0.3002 0 0.3>)
     Alpha=&MakeFloatFromInt(Int=$User.Variant)
    )
    ScaleA=true
    ScaleRGB=true
   )
   &DynamicMaterialParameters(
    Index 0 Param 1=200
    Index 0 Param 2=1
    Index 0 Param 3=1
    Index 0 Param 4=1
    Module.Index0Float0Write=true
    Module.Index0Float1Write=true
    Module.Index0Float2Write=true
    Module.Index0Float3Write=true
   )
   &CameraOffset(Camera Offset Amount=-200)
  ]
  rendererClasses [&NiagaraMeshRendererProperties]
  renderers [
   {rendererClass=&NiagaraMeshRendererProperties, rendererIndex=0}
   {rendererClass=&NiagaraMeshRendererProperties, rendererIndex=1}
  ]
 }
 Rings {
  emitterUpdateScript [
   &EmitterState()
   &SpawnRate(SpawnRate=5, Use Spawn Probability=false)
  ]
  particleSpawnScript [&InitializeParticle(Lifetime=2, Mesh Scale=<2.27 2 1>, Mesh Scale Mode=::NewEnumerator4, UsePositionOffset=false, Write Material Random=true)]
  particleUpdateScript [
   &ParticleState()
   /SetVariables(Particles.Position=&VectorFromCurve(VectorCurve=DataInterface() {propertyValues {
       XCurve {keys [<at 0 1>,<at 1 0>]}
       YCurve {keys [<at 0 1>,<at 1 0>]}
       ZCurve {keys [
         <at 0 5500>
         {arriveTangent=-4716.4663, interpMode=RCIM_Cubic, leaveTangent=-4716.4663, tangentMode=RCTM_User, time=0.5, value=1000}
         {arriveTangent=-1118.5477, tangentMode=RCTM_User, time=1, value=0}
        ]}
      }})) {isSetParameters=true}
   &ScaleMeshSize(Scale Factor=&VectorFromCurve(VectorCurve=DataInterface() {propertyValues {
       XCurve {keys [<at 0 1>,<at 1 1>]}
       YCurve {keys [<at 0 1>,<at 1 1>]}
       ZCurve {keys [
         {tangentMode=RCTM_User, time=0, value=1}
         {arriveTangent=-0.0405, interpMode=RCIM_Cubic, leaveTangent=-0.0405, tangentMode=RCTM_User, time=0.7009, value=1}
         {arriveTangent=-8.0089, interpMode=RCIM_Cubic, leaveTangent=-8.0089, tangentMode=RCTM_User, time=1, value=0}
        ]}
      }}))
   &ScaleColor(
    Scale Alpha=&FloatFromCurve(FloatCurve=DataInterface() {propertyValues {Curve {keys [
         {tangentMode=RCTM_User, time=0, value=1}
         {arriveTangent=-1.5952, interpMode=RCIM_Cubic, leaveTangent=-1.5952, tangentMode=RCTM_User, time=0.5, value=1}
         {arriveTangent=-2.1542, interpMode=RCIM_Cubic, leaveTangent=-2.1542, tangentMode=RCTM_User, time=1, value=0}
        ]}}})
    Scale RGB=&Lerp_Vector(
     A=&MakeVectorFromLinearColorRGB(Color=<rgba 0 0.4762 1 1>)
     B=&MakeVectorFromLinearColorRGB(Color=<rgba 0.7016 0.3638 0 0.5>)
     Alpha=&MakeFloatFromInt(Int=$User.Variant)
    )
    ScaleA=true
    ScaleRGB=true
   )
  ]
  rendererClasses [&NiagaraMeshRendererProperties]
  renderers [{rendererClass=&NiagaraMeshRendererProperties, rendererIndex=0}]
 }
 SquaresLoop {
  simTarget=GPUComputeSim
  emitterUpdateScript [
   &EmitterState()
   &SpawnRate(
    SpawnRate=&FloatFromCurve(
     FloatCurve=DataInterface() {propertyValues {Curve {keys [
         {arriveTangent=0.093, interpMode=RCIM_Cubic, leaveTangent=0.093, tangentMode=RCTM_User, time=0, value=0}
         {arriveTangent=3.6494, interpMode=RCIM_Cubic, leaveTangent=3.6494, tangentMode=RCTM_User, time=1, value=1}
        ]}}}
     CurveIndex=$Engine.System.Age
     Scale Curve=100
    )
    Use Spawn Probability=false
   )
  ]
  particleSpawnScript [
   &InitializeParticle(
    Lifetime=&RandomRangeFloat(Minimum=0.2, Maximum=0.3, Evaluation Type=::NewEnumerator1)
    Sprite Size=&RandomRangeVector2D(Minimum=<2.5 5>, Maximum=<3 6>, Evaluation Type=::NewEnumerator1)
    Sprite Size Mode=::NewEnumerator4
    UsePositionOffset=false
    Write Material Random=true
   )
   &ShapeLocation(
    Cylinder Height=1000
    Cylinder Radius=58
    MaskSpawnGroup=false
    Offset=<0 0 600>
    Offset Mode=::NewEnumerator0
    Override Seed=false
    Rotation Axis=<0 0 1>
    Rotation Mode=::NewEnumerator1
    Shape Primitive=::NewEnumerator1
    Surface Only=true
    UseOwnerScale=true
   )
   &AddVelocity(Velocity=&RandomRangeVector(Minimum=<0 0 -1700>, Maximum=<0 0 -1500>, Evaluation Type=::NewEnumerator1))
  ]
  particleUpdateScript [
   &ParticleState()
   &ScaleSpriteSize(
    Scale Factor=&Vector2DFromCurve(
     Vector2Curve=DataInterface() {propertyValues {
       XCurve {keys [{arriveTangent=-4.1323, interpMode=RCIM_Cubic, time=0, value=2}, <at 0.8 1>]}
       YCurve {keys [{arriveTangent=-13.0436, interpMode=RCIM_Cubic, time=0, value=5}, <at 0.8 1>]}
      }}
     ScaleCurve=<0.25 0.15>
    )
    Scale Sprite Size Mode=::NewEnumerator2
   )
   &ScaleColor(
    Scale Alpha=&FloatFromCurve(
     FloatCurve=DataInterface() {propertyValues {Curve {keys [{arriveTangent=5.9709, interpMode=RCIM_Cubic, time=0, value=0}, <at 0.2 0.5 RCIM_Cubic>, <at 1 0>]}}}
     Scale Curve=0.75
    )
    Scale RGB=&Lerp_Vector(
     A=&MakeVectorFromLinearColorRGB(Color=<rgba 0 0.4762 1 1>)
     B=&MakeVectorFromLinearColorRGB(Color=<rgba 1 0.5185 0 1>)
     Alpha=&MakeFloatFromInt(Int=$User.Variant)
    )
    ScaleA=true
    ScaleRGB=true
   )
   &Drag(Drag=&Lerp_Float(A=0, B=6, Alpha=0.5), Use Linear Drag=true, Use Rotational Drag=true)
   &SolveForcesAndVelocity(Clamp Velocity=false, Limit Acceleration=false)
   &ScaleSpriteSizeBySpeed(Max Scale Factor=<5 25>, Source Velocity:Vector3f=&Multiply_Vector(A=<0 0.05 1>, B=$Particles.Velocity))
  ]
  rendererClasses [&NiagaraSpriteRendererProperties]
  renderers [{rendererClass=&NiagaraSpriteRendererProperties}]
 }
}
summary {
 systemName=NS_SCMachine_SpawnWarmup
 emitters [
  {emitterName=PickupBeamDark, rendererClasses [&NiagaraMeshRendererProperties]}
  {emitterName=SquaresLoop, simTarget=GPUComputeSim, rendererClasses [&NiagaraSpriteRendererProperties]}
  {emitterName=Beam, rendererClasses [&NiagaraMeshRendererProperties]}
  {emitterName=Rings, rendererClasses [&NiagaraMeshRendererProperties]}
 ]
 userVariables [User.Variant {type=NiagaraInt32}]
}
