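; Event-triggered passive gameplay ability that drops a player's reboot card pickup for the SCMachine resurrection system.
; It skips spawning when event tags report no resurrector, reboot is disabled, a block tag is present, or the gamemode has respawns on.
; Spawn requires the drop setting, a participating pawn and valid controller, placing the card at a Z offset.
; Scalable-float auto-collect settings can relocate it to a living squadmate, either forced or when below the ocean actor.
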(event Ability|EventActivateAbilityFromEvent
  (params (EventData :type Gameplay%20Event%20Data%20Structure%20%28by%20ref%29))
  (bind (_eventtag _instigator _target _optionalobject _optionalobject2 _contexthandle _instigatortags _targettags) (Utilities|Struct|BreakGameplayEventData EventData))
  (CallFunction|SaveOutPlayer)
  (if (not (GameplayTags|HasTag _targettags "(TagName=\"DBNO.NoResurrectorAvailable\")"))
    (Utilities|FlowControl|Sequence
      (:then_0
        (bind _benabled (CallFunction|CheckResurrectionActive))
        (bind _ischipspawnenabled (CallFunction|CheckPlayerChipSpawnEnabled))
        (if (and _ischipspawnenabled _benabled)
          (Utilities|CheckforRespawnGamemode
            (:RespawnsOff
              (CallFunction|Spawn)))))
      (:then_1
        (Ability|EndAbility)))))

(event Custom|Spawn
  (bind _playerpawn (Variables|Default|GetPlayerPawn))
  (bind _returnvalue (Pawn|GetController _playerpawn))
  (bind (_returnvalue_location _returnvalue_rotation) (Math|Transform|BreakTransform (Transformation|GetActorTransform _playerpawn)))
  ;[1 1 1] Spawn
  (if (ScalableFloat|GetValueasBoolwithContext (Variables|Default|GetDropRebootCards))
    (if (Class|FortPlayerPawn|IsParticipating _playerpawn)
      (Utilities|IsValid _returnvalue
        (:Is%20Valid
          (if (CallFunction|AutoCollectCardsOverrideCheck)
            (bind (_autocollect _teamate_transform) (CallFunction|AutoCollectRebootCard (Variables|Default|GetPlayerPawn) (.z _returnvalue_location)))
            (else
              _))
          (Game|SpawnActorfromClass "/Game/Athena/Items/EnvironmentalItems/SCMachine/BGA_Athena_SCMachine_Pickup.BGA_Athena_SCMachine_Pickup_C" (select _autocollect _teamate_transform (Math|Transform|MakeTransform (+ _returnvalue_location (Variables|Default|GetSpawnZOffset)) _returnvalue_rotation)) "AlwaysSpawn" "OverrideRootScale" _returnvalue (Team|GetActorTeam _playerpawn) _returnvalue 0 0))) ; This is to make sure it was a valid player and not any kind of AI
    )))

(fn AutoCollectCardsOverrideCheck
  (bind _shouldautocollectcards_override (Variables|Default|GetShouldAutoCollectCards_Override))
  (bind _returnvalue (ScalableFloat|GetValueasBool _shouldautocollectcards_override))
  (Development|PrintString (Utilities|String|ToString(Integer) (ScalableFloat|GetValueasInteger _shouldautocollectcards_override)))
  (Development|PrintString (Utilities|String|ToString(Boolean) _returnvalue))
  (bind _output_get (Variables|Default|SetAutoCollectCardsOverride (and (ScalableFloat|IsSettingCurveValid _shouldautocollectcards_override) _returnvalue)))
  (return (or _output_get (and (ScalableFloat|IsSettingCurveValid (Variables|Default|GetShouldAutoCollectCards_OceanKill)) (ScalableFloat|GetValueasBool (Variables|Default|GetShouldAutoCollectCards_OceanKill))))))

(fn AutoCollectRebootCard
  (params (Pawn :type Fort%20Player%20Pawn%20Athena%20Object%20Reference) (ZPos :type Float%20%28double-precision%29))
  (Utilities|FlowControl|Sequence
    (:then_0
      (bind _returnvalue (Actor|GetActorOfClass "/Game/Athena/Asteria/Environments/Water/Asteria_Waterbody_Ocean_Parent.Asteria_Waterbody_Ocean_Parent_C"))
      (bind (_returnvalue_x _returnvalue_y _returnvalue_z) (Math|Vector|BreakVector (Transformation|GetActorLocation _returnvalue)))
      (if (or (<= ZPos _returnvalue_z) (Variables|Default|GetAutoCollectCardsOverride))
        (bind _asfort_player_state_athena (Utilities|Casting|CastToFortPlayerStateAthena (Variables|/Script/Engine.Pawn|GetPlayerState Pawn)))
        (bind _returnvalue_2 (Class|FortGameStateAthena|GetSquadMemberPlayerStates (Fortnite|Athena|GameState|GetGameStateAthenafromContext Pawn) _asfort_player_state_athena false))
        (for _array_element _returnvalue_2
          (bind _asplayer_pawn_athena (Utilities|Casting|CastToPlayerPawn_Athena (PlayerState|GetPawn _array_element)))
          (if (not (FortPawn|IsDead _asplayer_pawn_athena))
            (return true (Math|Transform|MakeTransform (Transformation|GetActorLocation _asplayer_pawn_athena) (Transformation|GetActorRotation _asplayer_pawn_athena)))))
        (else
          (return false 0))))
    (:then_1
      (return false 0))))

(fn CheckResurrectionActive
  (bind _returnvalue (Game|GetGameState))
  (bind _returnvalue_2 (Actor|GetComponentbyClass _returnvalue "/Script/FortniteGame.FortGameStateComponent_Reboot"))
  ;[1 1 1] Check if the playlist is a valid playlist from our list or hotfixed off.
  (Utilities|IsValid _returnvalue
    (:Is%20Valid
      (Utilities|IsValid _returnvalue_2
        (:Is%20Valid
          (bind _returnvalue_3 (Resurrection|IsResurrectionEnabled _returnvalue_2 (Variables|Default|GetPlayerPawn)))
          (return _returnvalue_3))))))

(fn CheckPlayerChipSpawnEnabled
  (bind _playerpawn (Variables|Default|GetPlayerPawn))
  (Utilities|IsValid _playerpawn
    (:Is%20Valid
      (return (not (GameplayTags|HasMatchingGameplayTag _playerpawn (Variables|Default|GetBlockChipSpawnTag)))))
    (:Is%20Not%20Valid
      ;[1 0.7931 0.116 1 10] Default to true if the player is invalid. At this point, the player should be valid, but in case there is logic that spawns the chip with an invalid player.
      (return true))))

(fn SaveOutPlayer
  (bind _asfort_player_pawn_athena (Utilities|Casting|CastToFortPlayerPawnAthena (Ability|GetAvatarActorfromActorInfo)))
  (Variables|Default|SetPlayerPawn _asfort_player_pawn_athena))
